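# Looping Niagara cosmetic effect for the male medium HighTower Tapas body, showing translucent blue-to-cyan star sprites attached to the character.
# A single Stars emitter randomly assigns each spawned particle one of four bone/socket indices, with randomized size, lifetime and material parameters.
# Each update re-samples that skeletal mesh location, adding a per-index offset chosen through nested integer-comparison selects.
# Sprites are offset toward the camera, and an alpha curve fades them in and out.
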NiagaraSystem /BRCosmetics/Characters/Player/Male/Medium/Bodies/M_MED_HighTower_Tapas/FX/NS_Tapas_LoopingFX_M
dependencies {
 usedDataInterfaces [&NiagaraDataInterfaceCurve, &NiagaraDataInterfaceSkeletalMesh]
 usedDynamicInputs []
 usedModules [
  &SystemState
  &self:SystemSpawnScript.SystemScriptSource.SystemScriptGraph.NiagaraNodeAssignment_0.SetVariables_#1
 ]
}
emitters {Stars {
  emitterUpdateScript [
   &EmitterState(Life Cycle Mode=::NewEnumerator1, Loop Duration=1, UseLoopDelay=false)
   &SpawnRate(Use Spawn Probability=false)
  ]
  particleSpawnScript [
   /Niagara/Modules/Spawn/Initialization/InitializeParticle(
    Color=&UniformRangedLinearColor(Minimum=<rgba 0.1703 0.4623 1 0.7>, Maximum=<rgba 0.1666 0.9901 1 0.7>)
    Lifetime=&UniformRangedFloat(Minimum=2, Maximum=3)
    Module.Write Color=true
    Module.Write Lifetime=true
    Module.Write Mass=true
    Module.Write Material Random=true
    Module.Write Position=false
    Module.Write Scale=false
    Module.Write SpriteRotation=false
    Module.Write SpriteSize=true
    Position=&SimulationPosition()
    Sprite Size=&Vector2DFromFloat(Value=&UniformRangedFloat(Minimum=10, Maximum=20))
   )
   /SetVariables(Particles.BoneSelection:NiagaraInt32=&UniformRangedInt(Minimum=0, Maximum=3)) {isSetParameters=true}
   &CameraOffset(Camera Offset Amount=28)
   &DynamicMaterialParameters(
    Index 0 Param 1=&UniformRangedFloat(Minimum=0.4, Maximum=0.7)
    Index 0 Param 2=&UniformRangedFloat(Minimum=0.8, Maximum=1.55)
    Index 0 Param 3=1
    Index 0 Param 4=1
    Module.Param0WriteEnabled=true
    Module.Param1WriteEnabled=true
    Module.Param2WriteEnabled=true
    Module.Param3WriteEnabled=true
   )
  ]
  particleUpdateScript [
   &SkeletalMeshLocation(
    Bone / Socket Index=$Particles.BoneSelection
    Bone Sampling Mode=::NewEnumerator4
    Inherit Velocity=false
    MaskSpawnGroup=false
    Sampled Position Offset=&MakeCustomVectorFromBool(
     Bool=&SetBoolByIntComparison(A=$Particles.BoneSelection, B=0, Comparison Type=::NewEnumerator0)
     True Vector=&MakeCustomVectorFromBool(
      Bool=&SetBoolByIntComparison(A=$Particles.BoneSelection, B=1, Comparison Type=::NewEnumerator0)
      True Vector=&MakeCustomVectorFromBool(
       Bool=&SetBoolByIntComparison(A=$Particles.BoneSelection, B=2, Comparison Type=::NewEnumerator0)
       True Vector=<5 0 1>
       False Vector=<-2 0 -1>
      )
      False Vector=<-2 6 -4>
     )
     False Vector=<6 0 5>
    )
    Sampled Position Offset Coordinate Space:ENiagaraMeshTransforms=::NewEnumerator3
    Skeletal Mesh=$System.Skeletal Mesh
   )
   &ParticleState()
   &SolveForcesAndVelocity(Clamp Velocity=false, Limit Acceleration=false)
   &ScaleColor(
    Scale Alpha=&FloatFromCurve(FloatCurve=DataInterface() {propertyValues {Curve {keys [
         <at 0 0 RCIM_Cubic>
         {interpMode=RCIM_Cubic, tangentMode=RCTM_User, time=0.2, value=1}
         {interpMode=RCIM_Cubic, tangentMode=RCTM_User, time=0.8, value=1}
         <at 1 0 RCIM_Cubic>
        ]}}})
    ScaleA=true
    ScaleRGB=true
   )
  ]
  rendererClasses [&NiagaraSpriteRendererProperties]
  renderers [{rendererClass=&NiagaraSpriteRendererProperties}]
 }}
summary {
 systemName=NS_Tapas_LoopingFX_M
 emitters [{emitterName=Stars, rendererClasses [&NiagaraSpriteRendererProperties]}]
 userVariables []
}
