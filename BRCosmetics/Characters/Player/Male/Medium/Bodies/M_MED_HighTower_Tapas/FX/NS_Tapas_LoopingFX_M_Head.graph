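# Looping Niagara cosmetic effect for the head of the Tapas (HighTower, medium male) character body, drawing a pair of glowing cyan eye sprites.
# One emitter bursts two short-lived sprites per loop, using spawn group to mirror them along Z around a fixed system-relative offset, with curve-driven alpha fades.
# User.bisHeadHidden drives each particle's VisibilityTag.
# The skeletal mesh bone-sampling placement is present but disabled.
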NiagaraSystem /BRCosmetics/Characters/Player/Male/Medium/Bodies/M_MED_HighTower_Tapas/FX/NS_Tapas_LoopingFX_M_Head
dependencies {usedDataInterfaces [&NiagaraDataInterfaceCurve], usedDynamicInputs [], usedModules [&SystemState]}
emitters {EyeGlows {
  emitterUpdateScript [
   &EmitterState(Inactive Response=::NewEnumerator1, Life Cycle Mode=::NewEnumerator1, Loop Duration=1, UseLoopDelay=false)
   &SpawnBurst_Instantaneous(Spawn Count=1, Use Spawn Probability=false)
   &SpawnBurst_Instantaneous(Spawn Count=1, Spawn Group=1, Use Spawn Probability=false) {module=SpawnBurst_Instantaneous001}
  ]
  particleSpawnScript [
   /Niagara/Modules/Spawn/Initialization/InitializeParticle(
    Color=<rgba 0.3388 2.5803 3 0.1>
    Lifetime=1.2
    Module.Write Color=true
    Module.Write Lifetime=true
    Module.Write Mass=true
    Module.Write Material Random=true
    Module.Write Position=false
    Module.Write Scale=false
    Module.Write SpriteRotation=false
    Module.Write SpriteSize=true
    Position=&SimulationPosition()
   )
   /SetVariables(Particles.BoneSelection:NiagaraInt32=&UniformRangedInt(Minimum=0, Maximum=1)) {enabled=false, isSetParameters=true}
   &SkeletalMeshLocation(
    Bone / Socket Index=$Particles.BoneSelection
    Inherit Velocity=false
    MaskSpawnGroup=false
    Mesh Sampling Type=::NewEnumerator5
    Sampled Position Offset Coordinate Space:ENiagaraMeshTransforms=::NewEnumerator3
    Skeletal Mesh=$User.SkeletalMesh
    Socket Sampling Mode=::NewEnumerator5
   ) {enabled=false}
   &CameraOffset(Camera Offset Amount=4)
   &DynamicMaterialParameters(
    Index 0 Param 1=1
    Index 0 Param 2=125
    Index 0 Param 3=1
    Index 0 Param 4=1
    Module.Param0WriteEnabled=true
    Module.Param1WriteEnabled=true
    Module.Param2WriteEnabled=true
    Module.Param3WriteEnabled=true
   )
   &SystemLocation(
    MaskSpawnGroup=false
    Offset=&MakeVector(
     X=4
     Y=8.4
     Z=&Lerp_Float(A=-3, B=3, Alpha=&MakeFloatFromInt(Int=&SpawnGroup()))
    )
   )
  ]
  particleUpdateScript [
   &ParticleState()
   &SolveForcesAndVelocity(Clamp Velocity=false, Limit Acceleration=false)
   &ScaleColor(
    Scale Alpha=&FloatFromCurve(FloatCurve=DataInterface() {propertyValues {Curve {keys [
         <at 0 0 RCIM_Cubic>
         <at 0.1 1>
         {arriveTangent=-1.1111, interpMode=RCIM_Cubic, leaveTangent=-1.1111, time=0.9, value=1}
         <at 1 0 RCIM_Cubic>
        ]}}})
    ScaleA=true
    ScaleRGB=true
   )
   /SetVariables(Particles.VisibilityTag=&MakeCustomIntFromBool(Bool=$User.bisHeadHidden, True INT=1, False INT=0)) {isSetParameters=true}
  ]
  rendererClasses [&NiagaraSpriteRendererProperties]
  renderers [{rendererClass=&NiagaraSpriteRendererProperties}]
 }}
summary {
 systemName=NS_Tapas_LoopingFX_M_Head
 emitters [{emitterName=EyeGlows, rendererClasses [&NiagaraSpriteRendererProperties]}]
 userVariables [
  User.bisHeadHidden {type=NiagaraBool}
  User.SkeletalMesh {
   type=NiagaraDataInterfaceSkeletalMesh
   defaultValue {
    struct=NiagaraExt_VariableValue_DataInterface
    value {dataInterface=&self:NiagaraDataInterfaceSkeletalMesh_0}
   }
  }
 ]
}
