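# Looping blue head effect for the male Tapas character cosmetic, with glows and electric strokes at two mirrored eye points.
# EyeGlows fades one sprite per side in and out, using spawn groups to pick the side.
# Electric_Strokes_Hands, despite its name, shares that placement, bursting short-lived, randomly flipped, growing strokes on random short loops, with brightness scaled by time of day.
# Both emitters drive particle visibility tags from User.bisHeadHidden.
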NiagaraSystem /BRCosmetics/Characters/Player/Male/Medium/Bodies/M_MED_HighTower_Tapas/FX/NS_Tapas_LoopingFX_M_Head_Blue
dependencies {usedDataInterfaces [&NiagaraDataInterfaceCurve], usedDynamicInputs [], usedModules [&SystemState]}
emitters {
 Electric_Strokes_Hands {
  emitterUpdateScript [
   &EmitterState(
    Inactive Response=::NewEnumerator1
    Life Cycle Mode=::NewEnumerator1
    Loop Duration=&RandomRangeFloat(Minimum=0.1, Maximum=0.3)
    Recalculate Duration Each Loop=true
    UseLoopDelay=false
   )
   &SpawnBurst_Instantaneous(Spawn Count=&UniformRangedInt(Minimum=1, Maximum=2), Use Spawn Probability=false)
   &SpawnBurst_Instantaneous(Spawn Count=&UniformRangedInt(Minimum=1, Maximum=2), Spawn Group=1, Use Spawn Probability=false) {module=SpawnBurst_Instantaneous001}
  ]
  particleSpawnScript [
   /Niagara/Modules/Spawn/Initialization/InitializeParticle(
    Color=<rgba 0.2184 2.8255 4 1>
    Lifetime=&UniformRangedFloat(Minimum=0.15, Maximum=0.35)
    Module.Write Color=true
    Module.Write Lifetime=true
    Module.Write Mass=true
    Module.Write Material Random=true
    Module.Write Position=true
    Module.Write Scale=false
    Module.Write SpriteRotation=true
    Module.Write SpriteSize=true
    Sprite Rotation=&Lerp_Float(A=-20, B=20, Alpha=&MakeFloatFromInt(Int=&SpawnGroup()))
    Sprite Size=&MakeVector2D(X=&UniformRangedFloat(Minimum=4, Maximum=7), Y=12)
   )
   &CameraOffset(Camera Offset Amount=4)
   &DynamicMaterialParameters(
    Index 0 Param 1=&UniformRangedFloat(Minimum=2.5, Maximum=3.5)
    Index 0 Param 3=&UniformRangedFloat(Minimum=0.2, Maximum=0.5)
    Module.Param0WriteEnabled=true
    Module.Param1WriteEnabled=false
    Module.Param2WriteEnabled=true
    Module.Param3WriteEnabled=true
   )
   /SetVariables(Particles.UVScale:Vector2f=&MakeVector2D(X=&UniformAOrBFloat(A=-1), Y=&UniformAOrBFloat(A=-1))) {isSetParameters=true}
   &SystemLocation(
    MaskSpawnGroup=false
    Offset=&MakeVector(
     X=4
     Y=8.4
     Z=&Lerp_Float(A=-2, B=2, Alpha=&MakeFloatFromInt(Int=&SpawnGroup()))
    )
   )
   /NiagaraFN/Sprite/FN_SpriteAlignmentAndFacing(
    SpriteAlignment:Vector3f=&Lerp_Vector(A=<0 0 -1>, B=<0 0 1>, Alpha=&MakeFloatFromInt(Int=&SpawnGroup()))
    SpriteFacing:Vector3f=<0 1 0>
   ) {module=SpriteAlignmentAndFacing}
  ]
  particleUpdateScript [
   &ParticleState()
   &ScaleSpriteSize(Scale Factor=&Vector2DFromFloat(Value=&FloatFromCurve(FloatCurve=DataInterface() {propertyValues {Curve {keys [{arriveTangent=1.9395, interpMode=RCIM_Cubic, leaveTangent=1.9395, tangentMode=RCTM_User, time=0, value=0.3}, <at 1 1 RCIM_Cubic>]}}})))
   &ScaleColor(
    Scale RGB=&VectorFromFloat(Value=&FNTimeOfDayScalar(Night Scale=15, Morning Scale=20, Day Scale=50, Evening Scale=40))
    ScaleA=true
    ScaleRGB=true
   )
   /SetVariables(Particles.VisibilityTag=&MakeCustomIntFromBool(Bool=$User.bisHeadHidden, True INT=1, False INT=0)) {isSetParameters=true, module=SetVariables_#1}
  ]
  rendererClasses [&NiagaraSpriteRendererProperties]
  renderers [{rendererClass=&NiagaraSpriteRendererProperties}]
 }
 EyeGlows {
  emitterUpdateScript [
   &EmitterState()
   &SpawnBurst_Instantaneous(Spawn Count=1, Use Spawn Probability=false)
   &SpawnBurst_Instantaneous(Spawn Count=1, Spawn Group=1, Use Spawn Probability=false) {module=SpawnBurst_Instantaneous001}
  ]
  particleSpawnScript [
   &InitializeParticle(
    Color=<rgba 0.2577 1.2375 2 1>
    Lifetime=1.2
    Module.Write Color=true
    Module.Write Lifetime=true
    Module.Write Mass=true
    Module.Write Material Random=true
    Module.Write Position=false
    Module.Write Scale=false
    Module.Write SpriteRotation=false
    Module.Write SpriteSize=true
    Position=&SimulationPosition()
    Sprite Size=<8 8>
   )
   &CameraOffset(Camera Offset Amount=4)
   &DynamicMaterialParameters(
    Index 0 Param 1=1
    Index 0 Param 2=125
    Index 0 Param 3=1
    Index 0 Param 4=1
    Module.Param0WriteEnabled=true
    Module.Param1WriteEnabled=true
    Module.Param2WriteEnabled=true
    Module.Param3WriteEnabled=true
   )
   &SystemLocation(
    MaskSpawnGroup=false
    Offset=&MakeVector(
     X=4
     Y=8.4
     Z=&Lerp_Float(A=-3, B=3, Alpha=&MakeFloatFromInt(Int=&SpawnGroup()))
    )
   )
  ]
  particleUpdateScript [
   &ParticleState()
   &ScaleColor(
    Scale Alpha=&FloatFromCurve(FloatCurve=DataInterface() {propertyValues {Curve {keys [
         <at 0 0 RCIM_Cubic>
         <at 0.1 1>
         {arriveTangent=-1.1111, interpMode=RCIM_Cubic, leaveTangent=-1.1111, time=0.9, value=1}
         <at 1 0 RCIM_Cubic>
        ]}}})
    ScaleA=true
    ScaleRGB=true
   )
   /SetVariables(Particles.VisibilityTag=&MakeCustomIntFromBool(Bool=$User.bisHeadHidden, True INT=1, False INT=0)) {isSetParameters=true, module=SetVariables_#1}
  ]
  rendererClasses [&NiagaraSpriteRendererProperties]
  renderers [{rendererClass=&NiagaraSpriteRendererProperties}]
 }
}
summary {
 systemName=NS_Tapas_LoopingFX_M_Head_Blue
 emitters [
  {emitterName=EyeGlows, rendererClasses [&NiagaraSpriteRendererProperties]}
  {emitterName=Electric_Strokes_Hands, rendererClasses [&NiagaraSpriteRendererProperties]}
 ]
 userVariables [User.bisHeadHidden {type=NiagaraBool}]
}
