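# Physics asset for the front loincloth of the male medium HighTower Tapas player body, simulating cloth motion with rigid body chains.
# Left and right chains of three dyn_skirt capsules link through free-swing, twist-locked constraints, so their swing limit angles are inactive.
# Each chain roots in a tiny MSTR anchor body, joined to its thigh with free rotation.
# Larger thigh capsules lock rotationally to a pelvis capsule, forming the body shapes around the cloth.
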PhysicsAsset /BRCosmetics/Characters/Player/Male/Medium/Bodies/M_MED_HighTower_Tapas/Meshes/M_MED_HighTower_Tapas_LoinCloth_Front_Physics_perfRework
bodies {
 dyn_skirt_fr_01_l [{shapeType=Capsule, length=10, radius=4, shapeName=dyn_skirt_fr_01_l_capsule, center=<6 0 0>, rotation=<rot 90 0 0>}]
 dyn_skirt_fr_01_l_MSTR [{shapeType=Capsule, length=0.01, radius=0.01, rotation=<rot 0 -90 0>}]
 dyn_skirt_fr_01_r [{shapeType=Capsule, length=10, radius=4, shapeName=dyn_skirt_fr_01_r_capsule, center=<6 0 0>, rotation=<rot 90 0 0>}]
 dyn_skirt_fr_01_r_MSTR [{shapeType=Capsule, length=0.01, radius=0.01, rotation=<rot 0 -90 0>}]
 dyn_skirt_fr_02_l [{shapeType=Capsule, length=10, radius=4, shapeName=dyn_skirt_fr_02_l_capsule, center=<6 0 0>, rotation=<rot 90 0 0>}]
 dyn_skirt_fr_02_r [{shapeType=Capsule, length=10, radius=4, shapeName=dyn_skirt_fr_02_r_capsule, center=<6 0 0>, rotation=<rot 90 0 0>}]
 dyn_skirt_fr_03_l [{shapeType=Capsule, length=10, radius=4, shapeName=dyn_skirt_fr_03_l_capsule, center=<6 0 0>, rotation=<rot 90 0 0>}]
 dyn_skirt_fr_03_r [{shapeType=Capsule, length=10, radius=4, shapeName=dyn_skirt_fr_03_r_capsule, center=<6 0 0>, rotation=<rot 90 0 0>}]
 pelvis [{shapeType=Capsule, length=10, radius=8.75, center=<-0.9243 -0.3604 0.0001>}]
 thigh_l [{shapeType=Capsule, length=32.198, radius=9.9, center=<-16.7937 -1.2981 -0.0013>, rotation=<rot 90 0 0>}]
 thigh_r [{shapeType=Capsule, length=31.698, radius=9.9, center=<16.7936 1.2981 0>, rotation=<rot 90 0 0>}]
}
constraints [
 {
  bone1Name=dyn_skirt_fr_01_l_MSTR
  bone2Name=thigh_l
  swing1LimitDegrees=45
  swing1Motion=Free
  swing2LimitDegrees=45
  swing2Motion=Free
  twistLimitDegrees=45
  twistMotion=Free
 }
 {
  bone1Name=dyn_skirt_fr_01_r_MSTR
  bone2Name=thigh_r
  swing1LimitDegrees=45
  swing1Motion=Free
  swing2LimitDegrees=45
  swing2Motion=Free
  twistLimitDegrees=45
  twistMotion=Free
 }
 {
  bone1Name=dyn_skirt_fr_01_l
  bone2Name=dyn_skirt_fr_01_l_MSTR
  swing1LimitDegrees=65
  swing1Motion=Free
  swing2LimitDegrees=65
  swing2Motion=Free
  twistLimitDegrees=35
  twistMotion=Locked
 }
 {
  bone1Name=dyn_skirt_fr_02_l
  bone2Name=dyn_skirt_fr_01_l
  swing1LimitDegrees=65
  swing1Motion=Free
  swing2LimitDegrees=65
  swing2Motion=Free
  twistLimitDegrees=25
  twistMotion=Locked
 }
 {
  bone1Name=dyn_skirt_fr_03_l
  bone2Name=dyn_skirt_fr_02_l
  swing1LimitDegrees=65
  swing1Motion=Free
  swing2LimitDegrees=65
  swing2Motion=Free
  twistLimitDegrees=25
  twistMotion=Locked
 }
 {
  bone1Name=dyn_skirt_fr_01_r
  bone2Name=dyn_skirt_fr_01_r_MSTR
  swing1LimitDegrees=65
  swing1Motion=Free
  swing2LimitDegrees=65
  swing2Motion=Free
  twistLimitDegrees=35
  twistMotion=Locked
 }
 {
  bone1Name=dyn_skirt_fr_02_r
  bone2Name=dyn_skirt_fr_01_r
  swing1LimitDegrees=65
  swing1Motion=Free
  swing2LimitDegrees=65
  swing2Motion=Free
  twistLimitDegrees=25
  twistMotion=Locked
 }
 {
  bone1Name=dyn_skirt_fr_03_r
  bone2Name=dyn_skirt_fr_02_r
  swing1LimitDegrees=65
  swing1Motion=Free
  swing2LimitDegrees=65
  swing2Motion=Free
  twistLimitDegrees=25
  twistMotion=Locked
 }
 {
  bone1Name=thigh_l
  bone2Name=pelvis
  swing1LimitDegrees=45
  swing1Motion=Locked
  swing2LimitDegrees=45
  swing2Motion=Locked
  twistLimitDegrees=45
  twistMotion=Locked
 }
 {
  bone1Name=thigh_r
  bone2Name=pelvis
  swing1LimitDegrees=45
  swing1Motion=Locked
  swing2LimitDegrees=45
  swing2Motion=Locked
  twistLimitDegrees=45
  twistMotion=Locked
 }
]
