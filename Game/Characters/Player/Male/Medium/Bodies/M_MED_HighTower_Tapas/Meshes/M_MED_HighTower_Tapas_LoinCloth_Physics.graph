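# Physics asset for the HighTower Tapas male medium player loincloth, defining jointed chains for its skirt bones and right-side chain drivers.
# Tiny "MSTR" root bodies anchor front, side and front-side chains to the thighs, and back and chain-driver chains to the pelvis.
# Box and capsule segments hang from these roots with angular limits.
# Cross-constraints tie neighbouring chains together: every back segment left-to-right, others only at their tips.
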PhysicsAsset /Game/Characters/Player/Male/Medium/Bodies/M_MED_HighTower_Tapas/Meshes/M_MED_HighTower_Tapas_LoinCloth_Physics
bodies {
 dyn_chaindriver_bk_01_r [{shapeType=Capsule, length=3.75, radius=3.75, center=<3.3731 0.0001 0>, rotation=<rot -90 -2.5305 2.5305>}]
 dyn_chaindriver_bk_01_r_MSTR [{shapeType=Capsule, length=0.01, radius=0.01, rotation=<rot 0 -90 0>}]
 dyn_chaindriver_bk_02_r [{shapeType=Capsule, length=3.75, radius=3.75, center=<3.3731 0.0001 0>, rotation=<rot -90 -2.5305 2.5305>}]
 dyn_chaindriver_fr_01_r [{shapeType=Capsule, length=3.75, radius=3.75, center=<3.3731 0.0001 0>, rotation=<rot -90 -2.5305 2.5305>}]
 dyn_chaindriver_fr_01_r_MSTR [{shapeType=Capsule, length=0.01, radius=0.01, rotation=<rot 0 -90 0>}]
 dyn_chaindriver_fr_02_r [{shapeType=Capsule, length=3.75, radius=3.75, center=<3.3731 0.0001 0>, rotation=<rot -90 -2.5305 2.5305>}]
 dyn_chainroot_MSTR [{shapeType=Capsule, length=0.01, radius=0.01, rotation=<rot 0 -90 0>}]
 dyn_skirt_01_l [{shapeType=Box, extentX=9.201, extentY=1.5123, extentZ=13.9922, center=<4.0214 0 0>, rotation=<rot 0 0 90>}]
 dyn_skirt_01_l_MSTR [{shapeType=Capsule, length=0.01, radius=0.01, rotation=<rot 0 -90 0>}]
 dyn_skirt_01_r [{shapeType=Box, extentX=9.201, extentY=1.5123, extentZ=13.9922, center=<4.0214 0 0>, rotation=<rot 0 0 90>}]
 dyn_skirt_01_r_MSTR [{shapeType=Capsule, length=0.01, radius=0.01, rotation=<rot 0 -90 0>}]
 dyn_skirt_02_l [{shapeType=Box, extentX=10.8242, extentY=1.7692, extentZ=13.9922, center=<4.3419 0 0>, rotation=<rot 0 0 90>}]
 dyn_skirt_02_r [{shapeType=Box, extentX=10.8242, extentY=1.7692, extentZ=13.9922, center=<4.3419 0 0>, rotation=<rot 0 0 90>}]
 dyn_skirt_03_l [{shapeType=Box, extentX=10.8242, extentY=3.7692, extentZ=13.9922, center=<4.3419 0 0>, rotation=<rot 0 0 90>}]
 dyn_skirt_03_r [{shapeType=Box, extentX=10.8242, extentY=3.7692, extentZ=13.9922, center=<4.3419 0 0>, rotation=<rot 0 0 90>}]
 dyn_skirt_bk_01_l [{shapeType=Box, extentX=11.7258, extentY=2.5523, extentZ=18.1146, center=<4.0214 0 0>, rotation=<rot 0 0 90>}]
 dyn_skirt_bk_01_l_MSTR [{shapeType=Capsule, length=0.01, radius=0.01, rotation=<rot 0 -90 0>}]
 dyn_skirt_bk_01_r [{shapeType=Box, extentX=11.7258, extentY=3.6729, extentZ=18.1146, center=<4.0214 0 0>, rotation=<rot 0 0 90>}]
 dyn_skirt_bk_01_r_MSTR [{shapeType=Capsule, length=0.01, radius=0.01, rotation=<rot 0 -90 0>}]
 dyn_skirt_bk_02_l [{shapeType=Box, extentX=11.7258, extentY=3.5523, extentZ=18.1146, center=<4.0214 0 0>, rotation=<rot 0 0 90>}]
 dyn_skirt_bk_02_r [{shapeType=Box, extentX=11.7258, extentY=4.6729, extentZ=18.1146, center=<4.0214 0 0>, rotation=<rot 0 0 90>}]
 dyn_skirt_bk_03_l [{shapeType=Box, extentX=11.7258, extentY=5.5523, extentZ=18.1146, center=<4.0214 0 0>, rotation=<rot 0 0 90>}]
 dyn_skirt_bk_03_r [{shapeType=Box, extentX=11.7258, extentY=6.6729, extentZ=18.1146, center=<4.0214 0 0>, rotation=<rot 0 0 90>}]
 dyn_skirt_fr_01_l [{shapeType=Box, extentX=9.201, extentY=3.0929, extentZ=17.1146, center=<4.0214 0 0>, rotation=<rot 0 0 90>}]
 dyn_skirt_fr_01_l_MSTR [{shapeType=Capsule, length=0.01, radius=0.01, rotation=<rot 0 -90 0>}]
 dyn_skirt_fr_01_r [{shapeType=Box, extentX=9.201, extentY=3.0929, extentZ=17.1146, center=<4.0214 0 0>, rotation=<rot 0 0 90>}]
 dyn_skirt_fr_01_r_MSTR [{shapeType=Capsule, length=0.01, radius=0.01, rotation=<rot 0 -90 0>}]
 dyn_skirt_fr_02_l [{shapeType=Box, extentX=9.701, extentY=4.0929, extentZ=17.1146, center=<4.4557 0 0>, rotation=<rot 0 0 90>}]
 dyn_skirt_fr_02_r [{shapeType=Box, extentX=9.701, extentY=4.0929, extentZ=17.1146, center=<4.4557 0 0>, rotation=<rot 0 0 90>}]
 dyn_skirt_fr_03_l [{shapeType=Box, extentX=9.701, extentY=5.0929, extentZ=17.1146, center=<4.4557 0 0>, rotation=<rot 0 0 90>}]
 dyn_skirt_fr_03_r [{shapeType=Box, extentX=9.701, extentY=5.0929, extentZ=17.1146, center=<4.4557 0 0>, rotation=<rot 0 0 90>}]
 dyn_skirtfrside_01_l [{shapeType=Capsule, length=9.8023, radius=1.6127, center=<2.8768 0.0994 0.0019>, rotation=<rot -1.0691 -88.0199 89.9629>}]
 dyn_skirtfrside_01_l_MSTR [{shapeType=Capsule, length=0.01, radius=0.01, rotation=<rot 0 -90 0>}]
 dyn_skirtfrside_01_r [{shapeType=Capsule, length=9.8023, radius=1.6127, center=<2.8768 0.0994 0.0019>, rotation=<rot -1.0691 -88.0199 89.9629>}]
 dyn_skirtfrside_01_r_MSTR [{shapeType=Capsule, length=5.108, radius=0.5, rotation=<rot 0 -90 0>}]
 dyn_skirtfrside_02_l [{shapeType=Capsule, length=9.8023, radius=1.6127, center=<2.8768 0.0994 0.0019>, rotation=<rot -1.0691 -88.0199 89.9629>}]
 dyn_skirtfrside_02_r [{shapeType=Capsule, length=9.8023, radius=1.6127, center=<2.8768 0.0994 0.0019>, rotation=<rot -1.0691 -88.0199 89.9629>}]
 dyn_skirtfrside_03_l [{shapeType=Capsule, length=9.1148, radius=2.9375, center=<2.8768 0.0994 0.0019>, rotation=<rot -1.0691 -88.0199 89.9629>}]
 dyn_skirtfrside_03_r [{shapeType=Capsule, length=9.1148, radius=2.9375, center=<2.8768 0.0994 0.0019>, rotation=<rot -1.0691 -88.0199 89.9629>}]
 pelvis [{shapeType=Capsule, length=10, radius=8.75, center=<-0.9243 -0.3604 0.0001>}]
 thigh_l [{shapeType=Capsule, length=32.198, radius=8.7544, center=<-16.7937 -1.2981 -0.0013>, rotation=<rot 90 0.0198 0.0197>}]
 thigh_r [{shapeType=Capsule, length=31.698, radius=9.2544, center=<16.7936 1.2981 0>, rotation=<rot 90 -0.0049 179.9951>}]
}
constraints [
 {
  bone1Name=dyn_skirt_01_l_MSTR
  bone2Name=thigh_l
  swing1LimitDegrees=45
  swing1Motion=Free
  swing2LimitDegrees=45
  swing2Motion=Free
  twistLimitDegrees=45
  twistMotion=Free
 }
 {
  bone1Name=dyn_skirt_fr_01_l_MSTR
  bone2Name=thigh_l
  swing1LimitDegrees=45
  swing1Motion=Free
  swing2LimitDegrees=45
  swing2Motion=Free
  twistLimitDegrees=45
  twistMotion=Free
 }
 {
  bone1Name=dyn_skirt_fr_01_r_MSTR
  bone2Name=thigh_r
  swing1LimitDegrees=45
  swing1Motion=Free
  swing2LimitDegrees=45
  swing2Motion=Free
  twistLimitDegrees=45
  twistMotion=Free
 }
 {
  bone1Name=dyn_skirt_01_r_MSTR
  bone2Name=thigh_r
  swing1LimitDegrees=45
  swing1Motion=Free
  swing2LimitDegrees=45
  swing2Motion=Free
  twistLimitDegrees=45
  twistMotion=Free
 }
 {
  bone1Name=dyn_skirt_bk_01_l_MSTR
  bone2Name=pelvis
  swing1LimitDegrees=45
  swing1Motion=Free
  swing2LimitDegrees=45
  swing2Motion=Free
  twistLimitDegrees=45
  twistMotion=Free
 }
 {
  bone1Name=dyn_skirt_bk_01_r_MSTR
  bone2Name=pelvis
  swing1LimitDegrees=45
  swing1Motion=Free
  swing2LimitDegrees=45
  swing2Motion=Free
  twistLimitDegrees=45
  twistMotion=Free
 }
 {
  bone1Name=dyn_chainroot_MSTR
  bone2Name=pelvis
  swing1LimitDegrees=45
  swing1Motion=Free
  swing2LimitDegrees=45
  swing2Motion=Free
  twistLimitDegrees=45
  twistMotion=Free
 }
 {
  bone1Name=dyn_chaindriver_bk_01_r_MSTR
  bone2Name=dyn_chainroot_MSTR
  swing1LimitDegrees=45
  swing1Motion=Free
  swing2LimitDegrees=45
  swing2Motion=Free
  twistLimitDegrees=45
  twistMotion=Free
 }
 {
  bone1Name=dyn_chaindriver_fr_01_r_MSTR
  bone2Name=dyn_chainroot_MSTR
  swing1LimitDegrees=45
  swing1Motion=Free
  swing2LimitDegrees=45
  swing2Motion=Free
  twistLimitDegrees=45
  twistMotion=Free
 }
 {bone1Name=dyn_skirt_01_l, bone2Name=dyn_skirt_01_l_MSTR, swing1LimitDegrees=5, swing2LimitDegrees=45, twistLimitDegrees=5}
 {bone1Name=dyn_skirt_02_l, bone2Name=dyn_skirt_01_l, swing1LimitDegrees=5, swing2LimitDegrees=45, twistLimitDegrees=5}
 {bone1Name=dyn_skirt_03_l, bone2Name=dyn_skirt_02_l, swing1LimitDegrees=5, swing2LimitDegrees=45, twistLimitDegrees=5}
 {bone1Name=dyn_skirt_fr_01_l, bone2Name=dyn_skirt_fr_01_l_MSTR, swing1LimitDegrees=65, swing2LimitDegrees=65, twistLimitDegrees=35}
 {bone1Name=dyn_skirt_fr_02_l, bone2Name=dyn_skirt_fr_01_l, swing1LimitDegrees=65, swing2LimitDegrees=65, twistLimitDegrees=25}
 {bone1Name=dyn_skirt_fr_03_l, bone2Name=dyn_skirt_fr_02_l, swing1LimitDegrees=65, swing2LimitDegrees=65, twistLimitDegrees=25}
 {bone1Name=dyn_skirt_fr_01_r, bone2Name=dyn_skirt_fr_01_r_MSTR, swing1LimitDegrees=65, swing2LimitDegrees=65, twistLimitDegrees=35}
 {bone1Name=dyn_skirt_fr_02_r, bone2Name=dyn_skirt_fr_01_r, swing1LimitDegrees=65, swing2LimitDegrees=65, twistLimitDegrees=25}
 {bone1Name=dyn_skirt_fr_03_r, bone2Name=dyn_skirt_fr_02_r, swing1LimitDegrees=65, swing2LimitDegrees=65, twistLimitDegrees=25}
 {bone1Name=dyn_skirt_01_r, bone2Name=dyn_skirt_01_r_MSTR, swing1LimitDegrees=5, swing2LimitDegrees=45, twistLimitDegrees=5}
 {bone1Name=dyn_skirt_02_r, bone2Name=dyn_skirt_01_r, swing1LimitDegrees=5, swing2LimitDegrees=45, twistLimitDegrees=5}
 {bone1Name=dyn_skirt_03_r, bone2Name=dyn_skirt_02_r, swing1LimitDegrees=5, swing2LimitDegrees=45, twistLimitDegrees=5}
 {bone1Name=dyn_skirt_bk_01_l, bone2Name=dyn_skirt_bk_01_l_MSTR, swing1LimitDegrees=15, swing2LimitDegrees=45, twistLimitDegrees=5}
 {bone1Name=dyn_skirt_bk_02_l, bone2Name=dyn_skirt_bk_01_l, swing1LimitDegrees=15, swing2LimitDegrees=45, twistLimitDegrees=5}
 {bone1Name=dyn_skirt_bk_03_l, bone2Name=dyn_skirt_bk_02_l, swing1LimitDegrees=15, swing2LimitDegrees=45, twistLimitDegrees=5}
 {bone1Name=dyn_skirt_bk_01_r, bone2Name=dyn_skirt_bk_01_r_MSTR, swing1LimitDegrees=15, swing2LimitDegrees=45, twistLimitDegrees=5}
 {bone1Name=dyn_skirt_bk_02_r, bone2Name=dyn_skirt_bk_01_r, swing1LimitDegrees=15, swing2LimitDegrees=45, twistLimitDegrees=5}
 {bone1Name=dyn_skirt_bk_03_r, bone2Name=dyn_skirt_bk_02_r, swing1LimitDegrees=15, swing2LimitDegrees=45, twistLimitDegrees=5}
 {
  bone1Name=thigh_l
  bone2Name=pelvis
  swing1LimitDegrees=45
  swing1Motion=Free
  swing2LimitDegrees=45
  swing2Motion=Free
  twistLimitDegrees=45
  twistMotion=Free
 }
 {
  bone1Name=thigh_r
  bone2Name=pelvis
  swing1LimitDegrees=45
  swing1Motion=Free
  swing2LimitDegrees=45
  swing2Motion=Free
  twistLimitDegrees=45
  twistMotion=Free
 }
 {
  bone1Name=dyn_skirt_bk_03_r
  bone2Name=dyn_skirt_bk_03_l
  swing1LimitDegrees=45
  swing1Motion=Free
  swing2LimitDegrees=45
  swing2Motion=Free
  twistLimitDegrees=45
  twistMotion=Free
 }
 {
  bone1Name=dyn_skirt_bk_02_r
  bone2Name=dyn_skirt_bk_02_l
  swing1LimitDegrees=45
  swing1Motion=Free
  swing2LimitDegrees=45
  swing2Motion=Free
  twistLimitDegrees=45
  twistMotion=Free
 }
 {
  bone1Name=dyn_skirt_bk_01_r
  bone2Name=dyn_skirt_bk_01_l
  swing1LimitDegrees=45
  swing1Motion=Free
  swing2LimitDegrees=45
  swing2Motion=Free
  twistLimitDegrees=45
  twistMotion=Free
 }
 {
  bone1Name=dyn_skirt_fr_03_l
  bone2Name=dyn_skirt_fr_03_r
  swing1LimitDegrees=45
  swing1Motion=Free
  swing2LimitDegrees=45
  swing2Motion=Free
  twistLimitDegrees=45
  twistMotion=Free
 }
 {bone1Name=dyn_skirtfrside_01_l, bone2Name=dyn_skirtfrside_01_l_MSTR, swing1LimitDegrees=5, swing2LimitDegrees=45, twistLimitDegrees=5}
 {bone1Name=dyn_skirtfrside_02_l, bone2Name=dyn_skirtfrside_01_l, swing1LimitDegrees=5, swing2LimitDegrees=45, twistLimitDegrees=5}
 {bone1Name=dyn_skirtfrside_03_l, bone2Name=dyn_skirtfrside_02_l, swing1LimitDegrees=5, swing2LimitDegrees=45, twistLimitDegrees=5}
 {
  bone1Name=dyn_skirtfrside_01_l_MSTR
  bone2Name=thigh_l
  swing1LimitDegrees=45
  swing1Motion=Free
  swing2LimitDegrees=45
  swing2Motion=Free
  twistLimitDegrees=45
  twistMotion=Free
 }
 {
  bone1Name=dyn_skirtfrside_01_r_MSTR
  bone2Name=thigh_r
  swing1LimitDegrees=45
  swing1Motion=Free
  swing2LimitDegrees=45
  swing2Motion=Free
  twistLimitDegrees=45
  twistMotion=Free
 }
 {bone1Name=dyn_skirtfrside_01_r, bone2Name=dyn_skirtfrside_01_r_MSTR, swing1LimitDegrees=5, swing2LimitDegrees=45, twistLimitDegrees=5}
 {bone1Name=dyn_skirtfrside_02_r, bone2Name=dyn_skirtfrside_01_r, swing1LimitDegrees=5, swing2LimitDegrees=45, twistLimitDegrees=5}
 {bone1Name=dyn_skirtfrside_03_r, bone2Name=dyn_skirtfrside_02_r, swing1LimitDegrees=5, swing2LimitDegrees=45, twistLimitDegrees=5}
 {
  bone1Name=dyn_skirtfrside_03_l
  bone2Name=dyn_skirt_03_l
  swing1LimitDegrees=60
  swing2LimitDegrees=60
  twistLimitDegrees=60
  twistMotion=Free
 }
 {bone1Name=dyn_chaindriver_bk_01_r, bone2Name=dyn_chaindriver_bk_01_r_MSTR, swing1LimitDegrees=15, swing2LimitDegrees=25, twistLimitDegrees=5}
 {bone1Name=dyn_chaindriver_bk_02_r, bone2Name=dyn_chaindriver_bk_01_r, swing1LimitDegrees=15, swing2LimitDegrees=45, twistLimitDegrees=5}
 {bone1Name=dyn_chaindriver_fr_01_r, bone2Name=dyn_chaindriver_fr_01_r_MSTR, swing1LimitDegrees=15, swing2LimitDegrees=25, twistLimitDegrees=2}
 {bone1Name=dyn_chaindriver_fr_02_r, bone2Name=dyn_chaindriver_fr_01_r, swing1LimitDegrees=15, swing2LimitDegrees=45, twistLimitDegrees=2}
 {
  bone1Name=dyn_chaindriver_fr_02_r
  bone2Name=dyn_chaindriver_bk_02_r
  swing1LimitDegrees=45
  swing1Motion=Free
  swing2LimitDegrees=45
  swing2Motion=Free
  twistLimitDegrees=45
  twistMotion=Free
 }
 {bone1Name=dyn_skirtfrside_03_r, bone2Name=dyn_skirt_03_r, swing1LimitDegrees=45, swing2LimitDegrees=45, twistLimitDegrees=45}
]
