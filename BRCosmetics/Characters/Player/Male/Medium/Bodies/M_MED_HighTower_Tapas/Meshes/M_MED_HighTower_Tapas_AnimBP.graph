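# Post-process style AnimBlueprint for the HighTower Tapas body, adding secondary motion to skirt, loincloth and chains over a pose copied from the parent mesh.
# Constraints and look-ats drive helper bones, then two pelvis-space RigidBody nodes simulate cloth, the first using overridden gravity.
# Later alpha-driven nodes adjust skirt, chain and muscle bones for movement, glider, long back bling and arm poses.
# Sprinting Alpha is inverted, reaching 1 only when grounded and not sprinting.
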AnimBlueprint /BRCosmetics/Characters/Player/Male/Medium/Bodies/M_MED_HighTower_Tapas/Meshes/M_MED_HighTower_Tapas_AnimBP
target_skeleton=/Game/Characters/Player/Male/Medium/Bodies/M_MED_HighTower_Tapas/Meshes/M_MED_HighTower_Tapas_Skeleton
graphs [
 AnimGraph {nodes [
   n1 = "Constraint - dyn_skirt_fr_01_l_MSTR (R)"(Alpha=1, ConstraintWeights_0=1) {
    alphaBoolBlend {blendOption=Linear}
    boneToModify {boneName=dyn_skirt_fr_01_l_MSTR}
    constraintSetup [{transformType=Rotation, targetBone {boneName=thigh_l}}]
    constraintWeights [1]
   }
   n2 = "Constraint - dyn_skirt_fr_01_r_MSTR (R)"(ComponentPose=@n1, Alpha=1, ConstraintWeights_0=1) {
    alphaBoolBlend {blendOption=Linear}
    boneToModify {boneName=dyn_skirt_fr_01_r_MSTR}
    constraintSetup [{transformType=Rotation, targetBone {boneName=thigh_r}}]
    constraintWeights [1]
   }
   n3 = "Constraint - dyn_skirt_01_l_MSTR (P)"(ComponentPose=@n2, Alpha=1, ConstraintWeights_0=1) {
    alphaBoolBlend {blendOption=Linear}
    boneToModify {boneName=dyn_skirt_01_l_MSTR}
    constraintSetup [{targetBone {boneName=pelvis}}]
    constraintWeights [1]
   }
   "Constraint - dyn_skirt_01_r_MSTR (P)"(ComponentPose=@n3, Alpha=1, ConstraintWeights_0=1) {
    alphaBoolBlend {blendOption=Linear}
    boneToModify {boneName=dyn_skirt_01_r_MSTR}
    constraintSetup [{targetBone {boneName=pelvis}}]
    constraintWeights [1]
   }
   "Transform (Modify) Bone - Bone: dyn_skirtfrside_01_l"(Alpha=Get CrouchingAlpha() -> (CrouchingAlpha:float), Translation="0,0,0", Rotation="5.000,0,0", Scale="1,1,1") {
    rotationMode=BMM_Additive
    rotationSpace=BCS_BoneSpace
    scaleSpace=BCS_BoneSpace
    translationMode=BMM_Additive
    translationSpace=BCS_BoneSpace
    alphaBoolBlend {blendOption=Linear}
    alphaScaleBiasClamp {bInterpResult=true}
    boneToModify {boneName=dyn_skirtfrside_01_l}
    rotation {pitch=5}
   }
   "Transform (Modify) Bone - Bone: dyn_skirtfrside_01_l_MSTR"(Alpha=Get CrouchingAlpha(), Translation="0.000,3.996,12", Rotation="0,0,0", Scale="1,1,1") {
    rotationMode=BMM_Additive
    rotationSpace=BCS_BoneSpace
    scaleSpace=BCS_BoneSpace
    translationMode=BMM_Additive
    translationSpace=BCS_BoneSpace
    alphaBoolBlend {blendOption=Linear}
    alphaScaleBiasClamp {bInterpResult=true}
    boneToModify {boneName=dyn_skirtfrside_01_l_MSTR}
    translation {y=3.996, z=12}
   }
   RigidBody!(
    ComponentPose
    Alpha=Get Skydive Alpha() -> (Skydive Alpha:double)
    OverrideWorldGravity:Vector=Get WindVector() -> (WindVector:Vector)
    ExternalForce
   ) -> (Pose) {
    bClampLinearTranslationLimitToRefPose=true
    bOverrideWorldGravity=true
    overridePhysicsAsset=/Game/Characters/Player/Male/Medium/Bodies/M_MED_HighTower_Tapas/Meshes/M_MED_HighTower_Tapas_LoinCloth_Physics
    simulationSpace=BaseBoneSpace
    alphaBoolBlend {blendOption=Linear}
    baseBoneRef {boneName=pelvis}
    componentAppliedLinearAccClamp=<3000 3000 3000>
    componentLinearAccScale=<0.3 0.3 0.5>
    componentLinearVelScale=<0.3 0.3 0.5>
    overrideWorldGravity {z=-980}
   }
   n4 = "Constraint - dyn_chaindriver_fr_01_r_MSTR (PPPP)"!(
    ComponentPose=Local To Component(LocalPose=Copy Pose From Mesh() {bUseAttachedParent=true})
    Alpha=1
    ConstraintWeights_0=0.4129
    ConstraintWeights_1:float=0.1207
    ConstraintWeights_2:float=0.0492
    ConstraintWeights_3:float=0.4172
   ) -> (Pose) {
    alphaBoolBlend {blendOption=Linear}
    boneToModify {boneName=dyn_chaindriver_fr_01_r_MSTR}
    constraintSetup [{targetBone {boneName=pelvis}}, {targetBone {boneName=spine_02}}, {targetBone {boneName=spine_03}}, {targetBone {boneName=spine_01}}]
    constraintWeights [1,1,1,1]
   }
   n5 = "Constraint - dyn_chaindriver_md_01_r_MSTR (R)"(ComponentPose=@n4, Alpha=1, ConstraintWeights_0=1) {
    alphaBoolBlend {blendOption=Linear}
    boneToModify {boneName=dyn_chaindriver_md_01_r_MSTR}
    constraintSetup [{transformType=Rotation, targetBone {boneName=thigh_twist_01_r}}]
    constraintWeights [1]
   }
   n6 = "Constraint - dyn_chaindriver_bk_01_r_MSTR (PPPP)"!(
    ComponentPose=@n5
    Alpha=1
    ConstraintWeights_0=0.5001
    ConstraintWeights_1:float=0.0961
    ConstraintWeights_2:float=0.0364
    ConstraintWeights_3:float=0.3673
   ) -> (Pose) {
    alphaBoolBlend {blendOption=Linear}
    boneToModify {boneName=dyn_chaindriver_bk_01_r_MSTR}
    constraintSetup [{targetBone {boneName=pelvis}}, {targetBone {boneName=spine_02}}, {targetBone {boneName=spine_03}}, {targetBone {boneName=spine_01}}]
    constraintWeights [1,1,1,1]
   }
   n7 = "Constraint - dyn_chaindriver_fr_01_r_MSTR (R)"(ComponentPose=@n6, Alpha=1, ConstraintWeights_0=1) {
    alphaBoolBlend {blendOption=Linear}
    boneToModify {boneName=dyn_chaindriver_fr_01_r_MSTR}
    constraintSetup [{transformType=Rotation, targetBone {boneName=thigh_r}}]
    constraintWeights [1]
   }
   n8 = "Constraint - dyn_chaindriver_bk_01_r_MSTR (R)"(ComponentPose=@n7, Alpha=0, ConstraintWeights_0=1) {
    alphaBoolBlend {blendOption=Linear}
    boneToModify {boneName=dyn_chaindriver_bk_01_r_MSTR}
    constraintSetup [{transformType=Rotation, targetBone {boneName=thigh_twist_01_r}}]
    constraintWeights [1]
   }
   n9 = Look At - Bone: dyn_chaindriver_fr_01_r_MSTR(ComponentPose=@n8, Alpha=0) {
    alphaBoolBlend {blendOption=Linear}
    boneToModify {boneName=dyn_chaindriver_fr_01_r_MSTR}
    lookAtLocation {x=0}
    lookAtTarget {boneReference {boneName=dyn_chaindriver_fr_03_r_AIM}}
    lookAt_Axis {axis=<1 0>}
    lookUp_Axis {axis {x=0, z=1}}
   }
   n10 = Look At - Bone: dyn_chaindriver_md_01_r_MSTR(ComponentPose=@n9, Alpha=1) {
    alphaBoolBlend {blendOption=Linear}
    boneToModify {boneName=dyn_chaindriver_md_01_r_MSTR}
    lookAtLocation {x=0}
    lookAtTarget {boneReference {boneName=dyn_chaindriver_md_03_r_AIM}}
    lookAt_Axis {axis=<1 0>}
    lookUp_Axis {axis {x=0, z=1}}
   }
   n11 = Look At - Bone: dyn_chaindriver_bk_01_r_MSTR(ComponentPose=@n10, Alpha=1) {
    alphaBoolBlend {blendOption=Linear}
    boneToModify {boneName=dyn_chaindriver_bk_01_r_MSTR}
    lookAtLocation {x=0}
    lookAtTarget {boneReference {boneName=dyn_chaindriver_bk_03_r_AIM}}
    lookAt_Axis {axis=<1 0>}
    lookUp_Axis {axis {x=0, z=1}}
   }
   n12 = Get bIsSkyDiving()
   n13 = "Transform (Modify) Bone - Bone: dyn_skirt_fr_01_r"!(ComponentPose=@n11, bAlphaBoolEnabled:bool=@n12, Translation="0,0,0", Rotation="0.000,0.000,-45", Scale="1,1,1") -> (Pose) {
    alphaInputType=Bool
    rotationMode=BMM_Additive
    alphaBoolBlend {blendInTime=0.1, blendOption=Linear, blendOutTime=0.1}
    boneToModify {boneName=dyn_skirt_fr_01_r}
    rotation {roll=-45}
   }
   n14 = "Transform (Modify) Bone - Bone: dyn_skirt_fr_01_l"!(ComponentPose=@n13, bAlphaBoolEnabled:bool=@n12, Translation="0,0,0", Rotation="0.000,0.000,-45", Scale="1,1,1") -> (Pose) {
    alphaInputType=Bool
    rotationMode=BMM_Additive
    alphaBoolBlend {blendInTime=0.1, blendOption=Linear, blendOutTime=0.1}
    boneToModify {boneName=dyn_skirt_fr_01_l}
    rotation {roll=-45}
   }
   n15 = RigidBody(ComponentPose=@n14, OverrideWorldGravity=Get Gravity Override()) {
    bClampLinearTranslationLimitToRefPose=true
    bOverrideWorldGravity=true
    overridePhysicsAsset=&/M_MED_HighTower_Tapas_LoinCloth_Physics_perfRework
    simulationSpace=BaseBoneSpace
    alphaBoolBlend {blendOption=Linear}
    baseBoneRef {boneName=pelvis}
    componentAppliedLinearAccClamp=<3000 3000 3000>
    componentLinearAccScale=<0.3 0.3 0.5>
    componentLinearVelScale=<0.3 0.3 0.5>
    overrideWorldGravity {z=-980}
   }
   n16 = RigidBody(ComponentPose=@n15, OverrideWorldGravity="0,0,-980") {
    bClampLinearTranslationLimitToRefPose=true
    bOverrideWorldGravity=true
    overridePhysicsAsset=&/M_MED_HighTower_Tapas_LoinCloth_Front_Physics_perfRework
    simulationSpace=BaseBoneSpace
    alphaBoolBlend {blendOption=Linear}
    baseBoneRef {boneName=pelvis}
    componentAppliedLinearAccClamp=<3000 3000 3000>
    componentLinearAccScale=<0.3 0.3 0.5>
    componentLinearVelScale=<0.3 0.3 0.5>
    overrideWorldGravity {z=-980}
   }
   n17 = Get Sprinting Alpha() -> (Sprinting Alpha:double)
   n18 = "Constraint - dyn_skirt_01_r_MSTR (P)"(ComponentPose=@n16, Alpha=@n17, ConstraintWeights_0=1) {
    alphaBoolBlend {blendOption=Linear}
    alphaScaleBiasClamp {bMapRange=true, outRange {max=0.2, min=0.1}}
    boneToModify {boneName=dyn_skirt_01_r_MSTR}
    constraintSetup [{targetBone {boneName=thigh_r}}]
    constraintWeights [1]
   }
   n19 = "Constraint - dyn_skirt_02_r (P)"(ComponentPose=@n18, Alpha=@n17, ConstraintWeights_0=1) {
    alphaBoolBlend {blendOption=Linear}
    alphaScaleBiasClamp {bMapRange=true, outRange {max=0.4, min=0.1}}
    boneToModify {boneName=dyn_skirt_02_r}
    constraintSetup [{targetBone {boneName=thigh_r}}]
    constraintWeights [1]
   }
   n20 = "Constraint - dyn_skirt_03_r (P)"(ComponentPose=@n19, Alpha=@n17, ConstraintWeights_0=1) {
    alphaBoolBlend {blendOption=Linear}
    alphaScaleBiasClamp {bMapRange=true, outRange {max=0.5, min=0.2}}
    boneToModify {boneName=dyn_skirt_03_r}
    constraintSetup [{targetBone {boneName=thigh_r}}]
    constraintWeights [1]
   }
   n21 = "Constraint - dyn_skirt_01_l_MSTR (P)"(ComponentPose=@n20, Alpha=@n17, ConstraintWeights_0=1) {
    alphaBoolBlend {blendOption=Linear}
    alphaScaleBiasClamp {bMapRange=true, outRange {max=0.2, min=0.1}}
    boneToModify {boneName=dyn_skirt_01_l_MSTR}
    constraintSetup [{targetBone {boneName=thigh_l}}]
    constraintWeights [1]
   }
   n22 = "Constraint - dyn_skirt_02_l (P)"(ComponentPose=@n21, Alpha=@n17, ConstraintWeights_0=1) {
    alphaBoolBlend {blendOption=Linear}
    alphaScaleBiasClamp {bMapRange=true, outRange {max=0.4, min=0.1}}
    boneToModify {boneName=dyn_skirt_02_l}
    constraintSetup [{targetBone {boneName=thigh_l}}]
    constraintWeights [1]
   }
   n23 = "Constraint - dyn_skirt_03_l (P)"(ComponentPose=@n22, Alpha=@n17, ConstraintWeights_0=1) {
    alphaBoolBlend {blendOption=Linear}
    alphaScaleBiasClamp {bMapRange=true, outRange {max=0.5, min=0.2}}
    boneToModify {boneName=dyn_skirt_03_l}
    constraintSetup [{targetBone {boneName=thigh_l}}]
    constraintWeights [1]
   }
   n24 = "Constraint - dyn_skirtfrside_02_r (P)"(ComponentPose=@n23, Alpha=@n17, ConstraintWeights_0=1) {
    alphaBoolBlend {blendOption=Linear}
    alphaScaleBiasClamp {bMapRange=true, outRange {max=0.5, min=0.3}}
    boneToModify {boneName=dyn_skirtfrside_02_r}
    constraintSetup [{targetBone {boneName=thigh_r}}]
    constraintWeights [1]
   }
   n25 = "Constraint - dyn_skirtfrside_03_r (P)"(ComponentPose=@n24, Alpha=@n17, ConstraintWeights_0=1) {
    alphaBoolBlend {blendOption=Linear}
    alphaScaleBiasClamp {bMapRange=true, outRange {max=0.9, min=0.5}}
    boneToModify {boneName=dyn_skirtfrside_03_r}
    constraintSetup [{targetBone {boneName=thigh_r}}]
    constraintWeights [1]
   }
   n26 = "Constraint - dyn_skirtfrside_02_l (P)"(ComponentPose=@n25, Alpha=@n17, ConstraintWeights_0=1) {
    alphaBoolBlend {blendOption=Linear}
    alphaScaleBiasClamp {bMapRange=true, outRange {max=0.5, min=0.3}}
    boneToModify {boneName=dyn_skirtfrside_02_l}
    constraintSetup [{targetBone {boneName=thigh_l}}]
    constraintWeights [1]
   }
   n27 = "Constraint - dyn_skirtfrside_03_l (P)"(ComponentPose=@n26, Alpha=@n17, ConstraintWeights_0=1) {
    alphaBoolBlend {blendOption=Linear}
    alphaScaleBiasClamp {bMapRange=true, outRange {max=0.9, min=0.5}}
    boneToModify {boneName=dyn_skirtfrside_03_l}
    constraintSetup [{targetBone {boneName=thigh_l}}]
    constraintWeights [1]
   }
   n28 = "Transform (Modify) Bone - Bone: dyn_skirt_bk_01_r_MSTR"(ComponentPose=@n27, Alpha=1, Translation="0,0,0", Rotation="3,0,0", Scale="1,1,1") {
    rotationMode=BMM_Additive
    rotationSpace=BCS_BoneSpace
    translationMode=BMM_Additive
    translationSpace=BCS_BoneSpace
    alphaBoolBlend {blendOption=Linear}
    boneToModify {boneName=dyn_skirt_bk_01_r_MSTR}
    rotation {pitch=3}
   }
   n29 = "Transform (Modify) Bone - Bone: dyn_skirt_bk_01_l_MSTR"(ComponentPose=@n28, Alpha=1, Translation="0,0,0", Rotation="3,0,0", Scale="1,1,1") {
    rotationMode=BMM_Additive
    rotationSpace=BCS_BoneSpace
    translationMode=BMM_Additive
    translationSpace=BCS_BoneSpace
    alphaBoolBlend {blendOption=Linear}
    boneToModify {boneName=dyn_skirt_bk_01_l_MSTR}
    rotation {pitch=3}
   }
   n30 = "Constraint - button_bk_01_r (PPPPP)"!(
    ComponentPose=@n29
    Alpha=1
    ConstraintWeights_0=0.3434
    ConstraintWeights_1:float=0.1917
    ConstraintWeights_2:float=0.1213
    ConstraintWeights_3:float=0.0243
    ConstraintWeights_4:float=0.3193
   ) -> (Pose) {
    alphaBoolBlend {blendOption=Linear}
    boneToModify {boneName=button_bk_01_r}
    constraintSetup [
     {targetBone {boneName=pelvis}}
     {targetBone {boneName=spine_02}}
     {targetBone {boneName=spine_03}}
     {targetBone {boneName=spine_04}}
     {targetBone {boneName=spine_01}}
    ]
    constraintWeights [1,1,1,1,1]
   }
   n31 = "Constraint - button_bk_03_r (PPPPP)"!(
    ComponentPose=@n30
    Alpha=1
    ConstraintWeights_0=0.1217
    ConstraintWeights_1:float=0.3214
    ConstraintWeights_2:float=0.2471
    ConstraintWeights_3:float=0.0694
    ConstraintWeights_4:float=0.2403
   ) -> (Pose) {
    alphaBoolBlend {blendOption=Linear}
    boneToModify {boneName=button_bk_03_r}
    constraintSetup [
     {targetBone {boneName=pelvis}}
     {targetBone {boneName=spine_02}}
     {targetBone {boneName=spine_03}}
     {targetBone {boneName=spine_04}}
     {targetBone {boneName=spine_01}}
    ]
    constraintWeights [1,1,1,1,1]
   }
   n32 = Look At - Bone: button_bk_02_r(ComponentPose=@n31, Alpha=1) {
    alphaBoolBlend {blendOption=Linear}
    boneToModify {boneName=button_bk_02_r}
    lookAtLocation {x=0}
    lookAtTarget {boneReference {boneName=button_bk_03_r}}
    lookAt_Axis {axis=<1 0>}
    lookUp_Axis {axis {x=0, z=1}}
   }
   n33 = Get ElbowRightAlpha() -> (ElbowRightAlpha:float)
   n34 = "Transform (Modify) Bone - Bone: bicep_02_r"(ComponentPose=@n32, Alpha=@n33, Translation="-6,0.000,-0.000", Rotation="0,0,0", Scale="1.000,1.4,1.000") {
    rotationMode=BMM_Additive
    rotationSpace=BCS_BoneSpace
    scaleMode=BMM_Additive
    scaleSpace=BCS_BoneSpace
    translationMode=BMM_Additive
    translationSpace=BCS_BoneSpace
    alphaBoolBlend {blendOption=Linear}
    boneToModify {boneName=bicep_02_r}
    scale {y=1.4}
    translation {x=-6}
   }
   n35 = "Transform (Modify) Bone - Bone: tricep_01_r"(ComponentPose=@n34, Alpha=@n33, Translation="0,0,0", Rotation="0,0,0", Scale="1.5,1.5,1.5") {
    rotationMode=BMM_Additive
    rotationSpace=BCS_BoneSpace
    scaleMode=BMM_Additive
    scaleSpace=BCS_BoneSpace
    translationMode=BMM_Additive
    translationSpace=BCS_BoneSpace
    alphaBoolBlend {blendOption=Linear}
    alphaScaleBiasClamp {bInterpResult=true}
    boneToModify {boneName=tricep_01_r}
    scale=<1.5 1.5 1.5>
   }
   n36 = "Transform (Modify) Bone - Bone: lowerarm_twist_02_adjust_r"(ComponentPose=@n35, Alpha=@n33, Translation="5.521,-0.000,0.000", Rotation="0,0,0", Scale="1,1,1") {
    rotationMode=BMM_Additive
    rotationSpace=BCS_BoneSpace
    scaleMode=BMM_Additive
    scaleSpace=BCS_BoneSpace
    translationMode=BMM_Additive
    translationSpace=BCS_BoneSpace
    alphaBoolBlend {blendOption=Linear}
    alphaScaleBiasClamp {bInterpResult=true}
    boneToModify {boneName=lowerarm_twist_02_adjust_r}
    translation {x=5.521}
   }
   n37 = "Transform (Modify) Bone - Bone: lowerarm_twist_02_adjust_l"(ComponentPose=@n36, Alpha=Get ElbowLeftAlpha() -> (ElbowLeftAlpha:float), Translation="5,-0.000,0.000", Rotation="0,0,0", Scale="1,1,1") {
    rotationMode=BMM_Additive
    rotationSpace=BCS_BoneSpace
    scaleMode=BMM_Additive
    scaleSpace=BCS_BoneSpace
    translationMode=BMM_Additive
    translationSpace=BCS_BoneSpace
    alphaBoolBlend {blendOption=Linear}
    alphaScaleBiasClamp {bInterpResult=true}
    boneToModify {boneName=lowerarm_twist_02_adjust_l}
    translation {x=5}
   }
   n38 = "Constraint - dyn_lat_measure_r (PPPPP)"!(
    ComponentPose=@n37
    Alpha=1
    ConstraintWeights_0=0.0002
    ConstraintWeights_1:float=0.0624
    ConstraintWeights_2:float=0.3939
    ConstraintWeights_3:float=0.5256
    ConstraintWeights_4:float=0.0179
   ) -> (Pose) {
    alphaBoolBlend {blendOption=Linear}
    boneToModify {boneName=dyn_lat_measure_r}
    constraintSetup [
     {targetBone {boneName=spine_01}}
     {targetBone {boneName=spine_02}}
     {targetBone {boneName=spine_03}}
     {targetBone {boneName=spine_04}}
     {targetBone {boneName=spine_05}}
    ]
    constraintWeights [1,1,1,1,1]
   }
   n39 = "Constraint - dyn_tricep_measure_r (PPPPPP)"!(
    ComponentPose=@n38
    Alpha=1
    ConstraintWeights_0=0.1888
    ConstraintWeights_1:float=0.183
    ConstraintWeights_2:float=0.3673
    ConstraintWeights_3:float=0.1979
    ConstraintWeights_4:float=0.0245
    ConstraintWeights_5:float=0.0386
   ) -> (Pose) {
    alphaBoolBlend {blendOption=Linear}
    boneToModify {boneName=dyn_tricep_measure_r}
    constraintSetup [
     {targetBone {boneName=tricep_01_r}}
     {targetBone {boneName=upperarm_r}}
     {targetBone {boneName=upperarm_twist_01_r}}
     {targetBone {boneName=upperarm_twist_02_r}}
     {targetBone {boneName=spine_03}}
     {targetBone {boneName=clavicle_r}}
    ]
    constraintWeights [1,1,1,1,1,1]
   }
   n40 = Get LatStretchRightAlpha() -> (LatStretchRightAlpha:float)
   n41 = "Transform (Modify) Bone - Bone: tricep_01_r"(ComponentPose=@n39, Alpha=@n40, Translation="0.000,-0.933,0.000", Rotation="0,0,0", Scale="0.6,0.6,0.6") {
    rotationMode=BMM_Additive
    rotationSpace=BCS_BoneSpace
    scaleMode=BMM_Additive
    scaleSpace=BCS_BoneSpace
    translationMode=BMM_Additive
    translationSpace=BCS_BoneSpace
    alphaBoolBlend {blendOption=Linear}
    boneToModify {boneName=tricep_01_r}
    scale=<0.6 0.6 0.6>
    translation {y=-0.933}
   }
   n42 = "Transform (Modify) Bone - Bone: tricep_02_r"(ComponentPose=@n41, Alpha=@n40, Translation="-0.000,-1.148,0.000", Rotation="0,0,0", Scale="1.000,0.8,1.000") {
    rotationMode=BMM_Additive
    rotationSpace=BCS_BoneSpace
    scaleMode=BMM_Additive
    scaleSpace=BCS_BoneSpace
    translationMode=BMM_Additive
    translationSpace=BCS_BoneSpace
    alphaBoolBlend {blendOption=Linear}
    boneToModify {boneName=tricep_02_r}
    scale {y=0.8}
    translation {y=-1.148}
   }
   n43 = "Transform (Modify) Bone - Bone: upperarm_twist_01_r"(ComponentPose=@n42, Alpha=@n40, Translation="0.000,0.000,-2", Rotation="0,0,0", Scale="1,1,1") {
    rotationMode=BMM_Additive
    rotationSpace=BCS_BoneSpace
    scaleMode=BMM_Additive
    scaleSpace=BCS_BoneSpace
    translationMode=BMM_Additive
    translationSpace=BCS_BoneSpace
    alphaBoolBlend {blendOption=Linear}
    boneToModify {boneName=upperarm_twist_01_r}
    translation {z=-2}
   }
   n44 = "Transform (Modify) Bone - Bone: lattissimus_00_r"(ComponentPose=@n43, Alpha=@n40, Translation="0.000,-2.934,-0.000", Rotation="0,0,0", Scale="1,1,1") {
    rotationMode=BMM_Additive
    rotationSpace=BCS_BoneSpace
    scaleMode=BMM_Additive
    scaleSpace=BCS_BoneSpace
    translationMode=BMM_Additive
    translationSpace=BCS_BoneSpace
    alphaBoolBlend {blendOption=Linear}
    boneToModify {boneName=lattissimus_00_r}
    translation {y=-2.934}
   }
   n45 = Get IsWearingCapeAlpha() -> (IsWearingCapeAlpha:float)
   n46 = "Constraint - dyn_skirt_bk_01_l (P)"(ComponentPose=@n44, Alpha=@n45, ConstraintWeights_0=1) {
    alphaBoolBlend {blendOption=Linear}
    boneToModify {boneName=dyn_skirt_bk_01_l}
    constraintSetup [{targetBone {boneName=pelvis}}]
    constraintWeights [1]
   }
   n47 = "Constraint - dyn_skirt_bk_01_r (P)"(ComponentPose=@n46, Alpha=@n45, ConstraintWeights_0=1) {
    alphaBoolBlend {blendOption=Linear}
    boneToModify {boneName=dyn_skirt_bk_01_r}
    constraintSetup [{targetBone {boneName=pelvis}}]
    constraintWeights [1]
   }
   n48 = Get HalfIsWearingCapeAlpha() -> (HalfIsWearingCapeAlpha:float)
   n49 = "Constraint - dyn_skirt_bk_02_r (R)"(ComponentPose=@n47, Alpha=@n48, ConstraintWeights_0=1) {
    alphaBoolBlend {blendOption=Linear}
    boneToModify {boneName=dyn_skirt_bk_02_r}
    constraintSetup [{transformType=Rotation, targetBone {boneName=thigh_r}}]
    constraintWeights [1]
   }
   n50 = "Constraint - dyn_skirt_bk_02_l (R)"(ComponentPose=@n49, Alpha=@n48, ConstraintWeights_0=1) {
    alphaBoolBlend {blendOption=Linear}
    boneToModify {boneName=dyn_skirt_bk_02_l}
    constraintSetup [{transformType=Rotation, targetBone {boneName=thigh_l}}]
    constraintWeights [1]
   }
   n51 = "Transform (Modify) Bone - Bone: ribs_01_r"(
    ComponentPose=@n50
    Alpha=Get TapasMontageAlpha() -> (TapasMontageAlpha:float)
    Translation="0.000,0.000,23"
    Rotation="0,0,0"
    Scale="1,1,1"
   ) {
    rotationMode=BMM_Additive
    rotationSpace=BCS_BoneSpace
    translationMode=BMM_Additive
    translationSpace=BCS_BoneSpace
    alphaBoolBlend {blendOption=Linear}
    alphaScaleBiasClamp {bInterpResult=true}
    boneToModify {boneName=ribs_01_r}
    translation {z=23}
   }
   n52 = Get tapasGliderAlpha() -> (tapasGliderAlpha:double)
   n53 = "Transform (Modify) Bone - Bone: dyn_skirt_fr_01_l"!(ComponentPose=@n51, Alpha:float=@n52, Translation="0,0,0", Rotation="-30.000,0.000,0.001", Scale="1,1,1") -> (Pose) {
    rotationMode=BMM_Additive
    rotationSpace=BCS_BoneSpace
    translationMode=BMM_Additive
    translationSpace=BCS_BoneSpace
    alphaBoolBlend {blendOption=Linear}
    alphaScaleBiasClamp {bInterpResult=true}
    boneToModify {boneName=dyn_skirt_fr_01_l}
    rotation {pitch=-30, roll=0.001}
   }
   n54 = "Transform (Modify) Bone - Bone: dyn_skirt_fr_01_r"!(ComponentPose=@n53, Alpha:float=@n52, Translation="0,0,0", Rotation="-24.594,-11.010,4.629", Scale="1,1,1") -> (Pose) {
    rotationMode=BMM_Additive
    rotationSpace=BCS_BoneSpace
    translationMode=BMM_Additive
    translationSpace=BCS_BoneSpace
    alphaBoolBlend {blendOption=Linear}
    alphaScaleBiasClamp {bInterpResult=true}
    boneToModify {boneName=dyn_skirt_fr_01_r}
    rotation=<rot -24.594 -11.0102 4.629>
   }
   n55 = "Transform (Modify) Bone - Bone: dyn_skirt_fr_02_r"(ComponentPose=@n54, Alpha=@n52, Translation="0,0,0", Rotation="-15.000,0.000,-0.001", Scale="1,1,1") {
    rotationMode=BMM_Additive
    rotationSpace=BCS_BoneSpace
    translationMode=BMM_Additive
    translationSpace=BCS_BoneSpace
    alphaBoolBlend {blendOption=Linear}
    alphaScaleBiasClamp {bInterpResult=true}
    boneToModify {boneName=dyn_skirt_fr_02_r}
    rotation {pitch=-15, roll=-0.001}
   }
   n56 = "Transform (Modify) Bone - Bone: dyn_skirt_fr_02_l"(ComponentPose=@n55, Alpha=@n52, Translation="0,0,0", Rotation="-15,0,0", Scale="1,1,1") {
    rotationMode=BMM_Additive
    rotationSpace=BCS_BoneSpace
    translationMode=BMM_Additive
    translationSpace=BCS_BoneSpace
    alphaBoolBlend {blendOption=Linear}
    alphaScaleBiasClamp {bInterpResult=true}
    boneToModify {boneName=dyn_skirt_fr_02_l}
    rotation {pitch=-15}
   }
   n57 = "Transform (Modify) Bone - Bone: dyn_skirt_fr_03_r"(ComponentPose=@n56, Alpha=@n52, Translation="0,0,0", Rotation="-24.594,-11.010,4.629", Scale="1,1,1") {
    rotationMode=BMM_Additive
    rotationSpace=BCS_BoneSpace
    translationMode=BMM_Additive
    translationSpace=BCS_BoneSpace
    alphaBoolBlend {blendOption=Linear}
    alphaScaleBiasClamp {bInterpResult=true}
    boneToModify {boneName=dyn_skirt_fr_03_r}
    rotation=<rot -24.594 -11.0102 4.629>
   }
   n58 = "Transform (Modify) Bone - Bone: dyn_skirt_fr_03_l"(ComponentPose=@n57, Alpha=@n52, Translation="0,0,0", Rotation="0,0,0", Scale="1,1,1") {
    rotationMode=BMM_Additive
    rotationSpace=BCS_BoneSpace
    translationMode=BMM_Additive
    translationSpace=BCS_BoneSpace
    alphaBoolBlend {blendOption=Linear}
    alphaScaleBiasClamp {bInterpResult=true}
    boneToModify {boneName=dyn_skirt_fr_03_l}
   }
   n59 = "Transform (Modify) Bone - Bone: dyn_chainroot_MSTR"(ComponentPose=@n58, Alpha=@n52, Translation="0,0,0", Rotation="-25,0.000,0.000", Scale="1,1,1") {
    rotationMode=BMM_Additive
    rotationSpace=BCS_BoneSpace
    translationMode=BMM_Additive
    translationSpace=BCS_BoneSpace
    alphaBoolBlend {blendOption=Linear}
    alphaScaleBiasClamp {bInterpResult=true}
    boneToModify {boneName=dyn_chainroot_MSTR}
    rotation {pitch=-25}
   }
   Output Pose(Result=Component To Local(ComponentPose=@n59)) {layerGroup=None}
   Get WindVector()
   Get Skydive Alpha()
  ]}
 EventGraph {nodes [
   n1 = ComplexSimSpaceFunction DEPRECATED!(
    execute
    CharacterPartAnimInstance:Custom Character Part Anim Instance Object Reference=Self-Reference()
    InterpSpeedMax:double=9
    PawnSpeedMax:double=20
    RInterpErrorTolerance:double=1
    DeltaTime:double=0
    DebugDraw:bool=false
    sim_space_rotation_in:Rotator=Get sim_space_rotation() -> (sim_space_rotation:Rotator)
    inComponentSpaceIn:bool=Get InComponentSpace() -> (InComponentSpace:bool)
    inComponentSpaceLastFrameIn:bool=Get InComponentSpaceLastFrame() -> (InComponentSpaceLastFrame:bool)
    lockedOnSimSpaceIn:bool=Get LockedOnSimSpace() -> (LockedOnSimSpace:bool)
    InterpSpeedIn:double=Get InterpSpeed() -> (InterpSpeed:double)
   ) -> (
    then
    sim_space_rotation:Rotator
    inComponentSpaceOut:bool
    inComponentSpaceLastFrameOut:bool
    lockedOnSimSpaceOut:bool
    InterpSpeedOut:double
   )
   > Set sim_space_rotation(sim_space_rotation:Rotator=@n1) -> (Output_Get:Rotator)
   > Set InComponentSpace(InComponentSpace:bool=@n1.inComponentSpaceOut) -> (Output_Get:bool)
   > Set InComponentSpaceLastFrame(InComponentSpaceLastFrame:bool=@n1.inComponentSpaceLastFrameOut) -> (Output_Get:bool)
   > Set LockedOnSimSpace(LockedOnSimSpace:bool=@n1.lockedOnSimSpaceOut) -> (Output_Get:bool)
   > Set InterpSpeed(InterpSpeed:double=@n1.InterpSpeedOut) -> (Output_Get:double)
   n2 = Cast To FortPlayerPawn(Object=GetOwningActor())
   n3 = Get ParachuteAttachment!(self:Fort Player Pawn Object Reference=@n2) -> (ParachuteAttachment:Fort Player Parachute Object Reference)
   > Is Valid(InputObject=@n3)
   n4 = Get ParachuteMesh!(self:Fort Player Parachute Object Reference=@n3) -> (ParachuteMesh:Skeletal Mesh Component Budgeted Object Reference)
   > Is Valid(InputObject=@n4)
   n5 = Get SkeletalMesh!(self:Skinned Mesh Component Object Reference=@n4) -> (SkeletalMesh:Skeletal Mesh Object Reference)
   > Is Valid(InputObject=@n5)
   Event BlueprintUpdateAnimation()
   > n6 = Sequence!(execute) -> (then_0, then_1, then_2:Exec)
   n7 = Get bIsSprinting()
   > n8 = Branch(Condition=AND Boolean(A=NOT Boolean(A=@n7), B=Get bIsOnGround() -> (bIsOnGround:bool)))
   > Set Sprinting Alpha(Sprinting Alpha:double=1) -> (Output_Get:double)
   > Set Sprint Back Cloth Alpha(Sprint Back Cloth Alpha:double=0) -> (Output_Get:double)
   Set Sprinting Alpha(execute=@n8.else, Sprinting Alpha=0)
   > Set Sprint Back Cloth Alpha(Sprint Back Cloth Alpha=1)
   Set tapasGliderAlpha(execute=@n6.then_1, tapasGliderAlpha:double="To Float (Boolean)"(InBool=Get bIsParachuteOpened())) -> (Output_Get:double)
   > n9 = Gravity Override State Function DEPRECATED(CharacterPartAnimInstance=Self-Reference(), GameTimeInSeconds=GetGameTimeInSeconds(), GravityOverrideParams=Get Gravity Override Params())
   > Set Gravity Override(Gravity Override=@n9)
   n10 = Branch(execute=@n6.then_2, Condition=Get bIsSkyDiving())
   > Set Skydive Alpha(Skydive Alpha:double=0.15) -> (Output_Get:double)
   Set Skydive Alpha(execute=@n10.else, Skydive Alpha=1)
   Event OnCharacterCustomizationLoaded()
   > Do Once()
   n11 = CheckForBackbling()
   > n12 = Set IsWearingCapeAlpha(IsWearingCapeAlpha:float=SelectFloat(A=1, B=0, bPickA=AND Boolean(A=@n11, B=@n11.bIsBackblingLong))) -> (Output_Get:float)
   > Set HalfIsWearingCapeAlpha(HalfIsWearingCapeAlpha:float=float / float(A=@n12, B=2) {is=title K2Node_CallFunction}) -> (Output_Get:float)
   "To Float (Boolean)"(InBool="Equal (Object)"(A=@n5, B=/BRCosmetics/Gadgets/Assets/VinderTech_GliderChute/Glider_HighTower_Tapas/Meshes/Glider_HighTower_Tapas) {is=title K2Node_CallFunction})
   n13 = GetCurrentActiveMontage!(self:Anim Instance Object Reference=GetAnimInstance(self=GetOwnerSkeletalMeshComponent())) -> (ReturnValue:Anim Montage Object Reference)
   PrintString!(execute, InString=GetDisplayName(Object=@n13), bPrintToScreen, bPrintToLog, TextColor, Duration=1000, Key) -> (then)
   Set isFrontendAndSkydivingAlpha(isFrontendAndSkydivingAlpha:double=SelectFloat(
     A=1
     B=0
     bPickA="Equal (Object)"(A=@n13, B=/Game/Animation/Game/MainPlayer/Combat/Melee/OneHanded/sythe/Sythe_Pickaxe_Harvesting_Combo_CMM_M)
    )) -> (Output_Get:double)
  ]}
]
variables [
 Gravity Override {type=Vector}
 Gravity Override Params {type=GravityOverrideParamsStruct}
 InComponentSpace {type=boolean}
 InComponentSpaceLastFrame {type=boolean}
 InterpSpeed {type=number}
 isFrontendAndSkydivingAlpha {type=number}
 isJumpingAlpha {type=number}
 isSkydiving {type=number}
 LockedOnSimSpace {type=boolean}
 NewVar_0 {type=AnimMontage, default=null}
 NewVar_1 {type=AnimMontage, default=/BRCosmetics/Animation/Game/MainPlayer/Emotes/Tapas/Tapas_Spin_M}
 NewVar_2 {type=AnimMontage, default=&Tapas_Spin_M}
 NewVar_3 {type=number}
 sim_space_rotation {type=Rotator}
 Skydive Alpha {type=number}
 Sprint Back Cloth Alpha {type=number}
 Sprinting Alpha {type=number}
 tapasGliderAlpha {type=number}
]
