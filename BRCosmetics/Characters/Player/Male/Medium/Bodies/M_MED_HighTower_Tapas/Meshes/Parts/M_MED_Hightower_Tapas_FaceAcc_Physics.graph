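# Physics asset for the HighTower Tapas FaceAcc part of a medium male player body, rigging long hair for physical motion.
# Capsule chains with mostly limited swing and twist form two front strands from the head, two side strands from neck_02, and three back strands from tiny MSTR anchor bodies.
# Angularly free constraints cross-link the outer back strands to the center one.
# Head, neck, clavicle, upper-arm and spine capsules supply body shapes around the hair.
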PhysicsAsset /BRCosmetics/Characters/Player/Male/Medium/Bodies/M_MED_HighTower_Tapas/Meshes/Parts/M_MED_Hightower_Tapas_FaceAcc_Physics
bodies {
 clavicle_l [{shapeType=Capsule, length=9.0452, radius=11.0848, shapeName=clavicle_l_capsule, center=<9 -0.8074 -0.492>, rotation=<rot 90 0 0>}]
 clavicle_r [{shapeType=Capsule, length=9.4751, radius=10.6549, shapeName=clavicle_r_capsule, center=<-9 0 0>, rotation=<rot 90 0 0>}]
 dyn_hairfront_bt_01_l [{shapeType=Capsule, length=7.5352, radius=1.5985, center=<2.6446 0 0.1144>, rotation=<rot 87.5114 179.9994 179.9994>}]
 dyn_hairfront_bt_01_r [{shapeType=Capsule, length=7.5352, radius=1.5985, center=<2.6446 0 0.1147>, rotation=<rot 87.5062 179.9995 179.9994>}]
 dyn_hairfront_bt_02_l [{shapeType=Capsule, length=7.5352, radius=1.5985, center=<2.6446 0 0.1144>, rotation=<rot 87.5184 179.9998 179.9996>}]
 dyn_hairfront_bt_02_r [{shapeType=Capsule, length=7.5352, radius=1.5985, center=<2.6446 0 0.1144>, rotation=<rot 87.511 179.9994 179.9994>}]
 dyn_hairfront_bt_03_l [{shapeType=Capsule, length=7.5352, radius=1.5985, center=<2.6446 0 0.1144>, rotation=<rot 87.5184 179.9998 179.9996>}]
 dyn_hairfront_bt_03_r [{shapeType=Capsule, length=7.5352, radius=1.5985, center=<2.6446 0 0.1144>, rotation=<rot 87.511 179.9994 179.9994>}]
 dyn_hairfront_bt_04_l [{shapeType=Capsule, length=7.5352, radius=1.8485, center=<2.6446 0 0.1144>, rotation=<rot 87.5184 179.9998 179.9996>}]
 dyn_hairfront_bt_04_r [{shapeType=Capsule, length=7.5352, radius=1.8485, center=<2.6446 0 0.1144>, rotation=<rot 87.511 179.9994 179.9994>}]
 dyn_hairfront_bt_05_l [{shapeType=Capsule, length=10, radius=1.85, center=<3.5443 0 0.1096>, rotation=<rot 88.2299 179.9996 179.9996>}]
 dyn_hairfront_bt_05_r [{shapeType=Capsule, length=10, radius=1.85, center=<3.5443 0 0.1098>, rotation=<rot 88.2122 179.9999 179.9999>}]
 dyn_hairlong_02_l [{shapeType=Capsule, length=5.0251, radius=2.3907, center=<3.1927 0 0>, rotation=<rot -90 0.0197 -0.0197>}]
 dyn_hairlong_02_r [{shapeType=Capsule, length=5.0251, radius=2.3907, center=<3.1927 0 0>, rotation=<rot -90 0.0197 -0.0197>}]
 dyn_hairlong_03_l [{shapeType=Capsule, length=4.7751, radius=2.6407, center=<3.1927 0 0>, rotation=<rot -90 0.0197 -0.0197>}]
 dyn_hairlong_03_r [{shapeType=Capsule, length=5.0251, radius=2.3907, center=<3.1927 0 0>, rotation=<rot -90 0.0197 -0.0197>}]
 dyn_hairlong_04_l [{shapeType=Capsule, length=4.7751, radius=2.6407, center=<3.0505 -0.8064 0>, rotation=<rot -90 179.1048 170.8953>}]
 dyn_hairlong_04_r [{shapeType=Capsule, length=5.0251, radius=2.3907, center=<3.1927 0 0>, rotation=<rot -90 0.0197 -0.0197>}]
 dyn_hairlong_bk_01 [{shapeType=Capsule, length=4.0251, radius=3.3907, center=<3.1927 0 0>, rotation=<rot -90 0.0197 -0.0197>}]
 dyn_hairlong_bk_01_MSTR [{shapeType=Capsule, length=0.01, radius=0.01}]
 dyn_hairlong_bk_01_l [{shapeType=Capsule, length=5.5251, radius=1.8907, center=<3.1927 0 0>, rotation=<rot -90 0.0197 -0.0197>}]
 dyn_hairlong_bk_01_l_MSTR [{shapeType=Capsule, length=0.01, radius=0.01}]
 dyn_hairlong_bk_01_r [{shapeType=Capsule, length=5.5251, radius=1.8907, center=<3.1927 0 0>, rotation=<rot -90 0.0197 -0.0197>}]
 dyn_hairlong_bk_01_r_MSTR [{shapeType=Capsule, length=0.01, radius=0.01}]
 dyn_hairlong_bk_02 [{shapeType=Capsule, length=5.2751, radius=2.1407, center=<3.1927 0 -0.3853>, rotation=<rot -90 0.0001 0>}]
 dyn_hairlong_bk_02_l [{shapeType=Capsule, length=4.1064, radius=2.1407, center=<2.5859 -0.3617 -0.4873>, rotation=<rot -85.0001 0.0002 -0.0001>}]
 dyn_hairlong_bk_02_r [{shapeType=Capsule, length=5.2751, radius=2.1407, center=<3.1927 0 0.5635>, rotation=<rot -90 180 -179.9999>}]
 dyn_hairlong_bk_03 [{shapeType=Capsule, length=4.7751, radius=2.6407, center=<3.1927 0 0>, rotation=<rot -90 0.0197 -0.0197>}]
 dyn_hairlong_bk_03_l [{shapeType=Capsule, length=3.4754, radius=4.0506, center=<3.8996 0.8192 -0.4306>, rotation=<rot -90 -144.7356 144.7357>}]
 dyn_hairlong_bk_03_r [{shapeType=Capsule, length=2.5734, radius=4.4219, center=<5 -0.3348 1.4685>, rotation=<rot -90 180 -179.9999>}]
 dyn_hairlong_bk_04 [{shapeType=Capsule, length=4.7466, radius=3.9203, center=<4 0 3>, rotation=<rot -90 0.0197 -0.0197>}]
 dyn_hairlong_bk_04_l [{shapeType=Capsule, length=2.2314, radius=4.0157, center=<2.6031 0 2.8033>, rotation=<rot -85 0.0002 -0.0001>}]
 dyn_hairlong_bk_04_r [{shapeType=Capsule, length=3.6501, radius=3.7657, center=<3.1927 0 1.9464>, rotation=<rot -90 180 -179.9999>}]
 dyn_hairlong_bk_05 [{shapeType=Capsule, length=3.4082, radius=4.0076, center=<5 0 3>, rotation=<rot -90 180 -179.9999>}]
 dyn_hairlong_bk_05_l [{shapeType=Capsule, length=2.3564, radius=3.8907, center=<2.5859 -0.282 3.4445>, rotation=<rot -90 -90 90.0001>}]
 dyn_hairlong_bk_05_r [{shapeType=Capsule, length=2.2751, radius=3.7657, center=<3.1927 2 2.8176>, rotation=<rot -90 0.0198 -0.0198>}]
 head [{shapeType=Capsule, length=4.7626, radius=8.841, center=<2.9176 1.5296 0>, rotation=<rot 90 -0.0198 -0.0198>}]
 neck_01 [{shapeType=Capsule, length=3.8265, radius=6.6876, center=<3.3927 0.0731 0>, rotation=<rot 90 -0.0099 -0.0099>}]
 neck_02 [{shapeType=Capsule, length=13.9551, radius=22.5499, shapeName=neck_02_capsule, center=<13.5853 20.5531 0>, rotation=<rot 0 -85.0002 -90>}]
 spine_04_COL [
  {shapeType=Capsule, length=18.25, radius=8.75, center=<5.2801 5.8521 -0.0001>, rotation=<rot -0.0005 -180 -180>}
  {shapeType=Capsule, length=15, radius=15, center=<0 -1.459 0.0001>}
 ]
 spine_05_COL2 [{shapeType=Capsule, length=33.7907, radius=11.1679, center=<2.6595 -0.9451 0.0001>, rotation=<rot -0.0005 -179.9999 -179.9999>}]
 upperarm_l [{shapeType=Capsule, length=21.2143, radius=9.5407, center=<8.8498 -1.3674 -2.6872>, rotation=<rot 0 -89.9995 100.0002>}]
 upperarm_r [{shapeType=Capsule, length=21.2143, radius=8.2907, center=<-9.1232 1.3673 1.1367>, rotation=<rot 0 -89.9995 -79.9996>}]
}
constraints [
 {bone1Name=dyn_hairlong_03_l, bone2Name=dyn_hairlong_02_l, swing1LimitDegrees=45, swing2LimitDegrees=5, twistLimitDegrees=5}
 {bone1Name=dyn_hairlong_04_l, bone2Name=dyn_hairlong_03_l, swing1LimitDegrees=45, swing2LimitDegrees=5, twistLimitDegrees=5}
 {bone1Name=dyn_hairlong_bk_01_l, bone2Name=dyn_hairlong_bk_01_l_MSTR, swing1LimitDegrees=5, swing2LimitDegrees=15, twistLimitDegrees=5}
 {bone1Name=dyn_hairlong_bk_02_l, bone2Name=dyn_hairlong_bk_01_l, swing1LimitDegrees=8, swing2LimitDegrees=30, twistLimitDegrees=8}
 {bone1Name=dyn_hairlong_bk_03_l, bone2Name=dyn_hairlong_bk_02_l, swing1LimitDegrees=8, swing2LimitDegrees=30, twistLimitDegrees=8}
 {bone1Name=dyn_hairlong_bk_04_l, bone2Name=dyn_hairlong_bk_03_l, swing1LimitDegrees=5, swing2LimitDegrees=25, twistLimitDegrees=5}
 {bone1Name=dyn_hairlong_bk_05_l, bone2Name=dyn_hairlong_bk_04_l, swing1LimitDegrees=5, swing2LimitDegrees=25, twistLimitDegrees=5}
 {bone1Name=dyn_hairlong_bk_01, bone2Name=dyn_hairlong_bk_01_MSTR, swing1LimitDegrees=5, swing2LimitDegrees=45, twistLimitDegrees=5}
 {bone1Name=dyn_hairlong_bk_02, bone2Name=dyn_hairlong_bk_01, swing1LimitDegrees=5, swing2LimitDegrees=45, twistLimitDegrees=5}
 {bone1Name=dyn_hairlong_bk_03, bone2Name=dyn_hairlong_bk_02, swing1LimitDegrees=5, swing2LimitDegrees=45, twistLimitDegrees=5}
 {bone1Name=dyn_hairlong_bk_04, bone2Name=dyn_hairlong_bk_03, swing1LimitDegrees=5, swing2LimitDegrees=45, twistLimitDegrees=5}
 {bone1Name=dyn_hairlong_bk_05, bone2Name=dyn_hairlong_bk_04, swing1LimitDegrees=5, swing2LimitDegrees=45, twistLimitDegrees=5}
 {bone1Name=dyn_hairlong_bk_01_r, bone2Name=dyn_hairlong_bk_01_r_MSTR, swing1LimitDegrees=5, swing2LimitDegrees=45, twistLimitDegrees=5}
 {
  bone1Name=dyn_hairlong_bk_02_r
  bone2Name=dyn_hairlong_bk_01_r
  swing1LimitDegrees=5
  swing2LimitDegrees=25
  twistLimitDegrees=5
  twistMotion=Locked
 }
 {
  bone1Name=dyn_hairlong_bk_03_r
  bone2Name=dyn_hairlong_bk_02_r
  swing1LimitDegrees=5
  swing2LimitDegrees=25
  twistLimitDegrees=5
  twistMotion=Locked
 }
 {
  bone1Name=dyn_hairlong_bk_04_r
  bone2Name=dyn_hairlong_bk_03_r
  swing1LimitDegrees=5
  swing2LimitDegrees=25
  twistLimitDegrees=5
  twistMotion=Locked
 }
 {bone1Name=dyn_hairlong_03_r, bone2Name=dyn_hairlong_02_r, swing1LimitDegrees=45, swing2LimitDegrees=5, twistLimitDegrees=5}
 {bone1Name=dyn_hairlong_04_r, bone2Name=dyn_hairlong_03_r, swing1LimitDegrees=45, swing2LimitDegrees=5, twistLimitDegrees=5}
 {bone1Name=dyn_hairfront_bt_01_r, bone2Name=head, swing1LimitDegrees=15, swing2LimitDegrees=45, twistLimitDegrees=5, twistMotion=Locked}
 {bone1Name=dyn_hairfront_bt_02_r, bone2Name=dyn_hairfront_bt_01_r, swing1LimitDegrees=15, swing2LimitDegrees=45, twistLimitDegrees=5}
 {bone1Name=dyn_hairfront_bt_03_r, bone2Name=dyn_hairfront_bt_02_r, swing1LimitDegrees=15, swing2LimitDegrees=45, twistLimitDegrees=5}
 {bone1Name=dyn_hairfront_bt_04_r, bone2Name=dyn_hairfront_bt_03_r, swing1LimitDegrees=45, swing2LimitDegrees=45, twistLimitDegrees=5}
 {bone1Name=dyn_hairfront_bt_05_r, bone2Name=dyn_hairfront_bt_04_r, swing1LimitDegrees=45, swing2LimitDegrees=45, twistLimitDegrees=5}
 {bone1Name=dyn_hairfront_bt_01_l, bone2Name=head, swing1LimitDegrees=15, swing2LimitDegrees=45, twistLimitDegrees=5, twistMotion=Locked}
 {bone1Name=dyn_hairfront_bt_02_l, bone2Name=dyn_hairfront_bt_01_l, swing1LimitDegrees=15, swing2LimitDegrees=45, twistLimitDegrees=5}
 {bone1Name=dyn_hairfront_bt_03_l, bone2Name=dyn_hairfront_bt_02_l, swing1LimitDegrees=15, swing2LimitDegrees=45, twistLimitDegrees=5}
 {bone1Name=dyn_hairfront_bt_04_l, bone2Name=dyn_hairfront_bt_03_l, swing1LimitDegrees=45, swing2LimitDegrees=45, twistLimitDegrees=5}
 {bone1Name=dyn_hairfront_bt_05_l, bone2Name=dyn_hairfront_bt_04_l, swing1LimitDegrees=45, swing2LimitDegrees=45, twistLimitDegrees=5}
 {
  bone1Name=dyn_hairlong_bk_01_l_MSTR
  bone2Name=head
  swing1LimitDegrees=45
  swing1Motion=Free
  swing2LimitDegrees=45
  swing2Motion=Free
  twistLimitDegrees=45
  twistMotion=Free
 }
 {
  bone1Name=dyn_hairlong_bk_01_MSTR
  bone2Name=head
  swing1LimitDegrees=45
  swing1Motion=Free
  swing2LimitDegrees=45
  swing2Motion=Free
  twistLimitDegrees=45
  twistMotion=Free
 }
 {
  bone1Name=dyn_hairlong_bk_01_r_MSTR
  bone2Name=head
  swing1LimitDegrees=45
  swing1Motion=Free
  swing2LimitDegrees=45
  swing2Motion=Free
  twistLimitDegrees=45
  twistMotion=Free
 }
 {
  bone1Name=head
  bone2Name=neck_02
  swing1LimitDegrees=45
  swing1Motion=Free
  swing2LimitDegrees=45
  swing2Motion=Free
  twistLimitDegrees=45
  twistMotion=Free
 }
 {
  bone1Name=dyn_hairlong_bk_05_l
  bone2Name=dyn_hairlong_bk_05
  swing1LimitDegrees=45
  swing1Motion=Free
  swing2LimitDegrees=45
  swing2Motion=Free
  twistLimitDegrees=45
  twistMotion=Free
 }
 {
  bone1Name=dyn_hairlong_bk_05_r
  bone2Name=dyn_hairlong_bk_04_r
  swing1LimitDegrees=5
  swing2LimitDegrees=25
  twistLimitDegrees=5
  twistMotion=Locked
 }
 {
  bone1Name=dyn_hairlong_bk_05_r
  bone2Name=dyn_hairlong_bk_05
  swing1LimitDegrees=45
  swing1Motion=Free
  swing2LimitDegrees=45
  swing2Motion=Free
  twistLimitDegrees=45
  twistMotion=Free
 }
 {
  bone1Name=upperarm_l
  bone2Name=clavicle_l
  swing1LimitDegrees=45
  swing1Motion=Free
  swing2LimitDegrees=45
  swing2Motion=Free
  twistLimitDegrees=45
  twistMotion=Free
 }
 {
  bone1Name=upperarm_r
  bone2Name=clavicle_r
  swing1LimitDegrees=45
  swing1Motion=Free
  swing2LimitDegrees=45
  swing2Motion=Free
  twistLimitDegrees=45
  twistMotion=Free
 }
 {
  bone1Name=dyn_hairlong_bk_03_l
  bone2Name=dyn_hairlong_bk_03
  swing1LimitDegrees=45
  swing1Motion=Free
  swing2LimitDegrees=45
  swing2Motion=Free
  twistLimitDegrees=45
  twistMotion=Free
 }
 {
  bone1Name=dyn_hairlong_bk_03_r
  bone2Name=dyn_hairlong_bk_03
  swing1LimitDegrees=45
  swing1Motion=Free
  swing2LimitDegrees=45
  swing2Motion=Free
  twistLimitDegrees=45
  twistMotion=Free
 }
 {
  bone1Name=dyn_hairlong_02_l
  bone2Name=neck_02
  swing1LimitDegrees=45
  swing1Motion=Free
  swing2LimitDegrees=45
  swing2Motion=Free
  twistLimitDegrees=45
  twistMotion=Free
 }
 {
  bone1Name=dyn_hairlong_02_r
  bone2Name=neck_02
  swing1LimitDegrees=45
  swing1Motion=Free
  swing2LimitDegrees=45
  swing2Motion=Free
  twistLimitDegrees=45
  twistMotion=Free
 }
 {
  bone1Name=neck_02
  bone2Name=neck_01
  swing1LimitDegrees=45
  swing1Motion=Free
  swing2LimitDegrees=45
  swing2Motion=Free
  twistLimitDegrees=45
  twistMotion=Free
 }
]
