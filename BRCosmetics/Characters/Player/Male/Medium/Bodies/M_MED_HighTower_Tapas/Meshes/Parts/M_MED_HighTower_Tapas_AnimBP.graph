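# Animation Blueprint for the HighTower Tapas body part, posing and simulating the character's hair.
# It copies the attached parent's pose, applies head-look-up Look Ats and bone constraints, then caches it.
# bRigidsEnabled selects between front-hair Look Ats and neck-relative RigidBody physics with wind gravity plus backbling and head-tilt corrective offsets.
# Events reduce simulation alpha while skydiving, reset dynamics afterward and detect backbling; several bone-offset chains remain unconnected.
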AnimBlueprint /BRCosmetics/Characters/Player/Male/Medium/Bodies/M_MED_HighTower_Tapas/Meshes/Parts/M_MED_HighTower_Tapas_AnimBP
target_skeleton=/Game/Characters/Player/Male/Medium/Bodies/M_MED_HighTower_Tapas/Meshes/Parts/M_MED_HighTower_Tapas_Skeleton
graphs [
 AnimGraph {nodes [
   "Transform (Modify) Bone - Bone: dyn_middle_hair_01_r"(Alpha=Get HeadLookUpAlpha() -> (HeadLookUpAlpha:float), Translation="0.000,0.000,5.678", Rotation="0,0,0", Scale="1,1,1") {
    rotationMode=BMM_Additive
    rotationSpace=BCS_BoneSpace
    scaleMode=BMM_Additive
    scaleSpace=BCS_BoneSpace
    translationMode=BMM_Additive
    translationSpace=BCS_BoneSpace
    alphaBoolBlend {blendOption=Linear}
    alphaScaleBiasClamp {bInterpResult=true}
    boneToModify {boneName=dyn_middle_hair_01_r}
    translation {z=5.678}
   }
   n1 = "Transform (Modify) Bone - Bone: dyn_hairlong_bk_02"(Alpha=1, Translation="0,0,0", Rotation="15,0.000,0.000", Scale="1,1,1") {
    rotationMode=BMM_Additive
    rotationSpace=BCS_BoneSpace
    translationMode=BMM_Additive
    translationSpace=BCS_BoneSpace
    alphaBoolBlend {blendOption=Linear}
    alphaScaleBiasClamp {bInterpResult=true}
    boneToModify {boneName=dyn_hairlong_bk_02}
    rotation {pitch=15}
   }
   n2 = "Transform (Modify) Bone - Bone: dyn_hairlong_bk_02_r"(ComponentPose=@n1, Alpha=1, Translation="0,0,0", Rotation="15,0.000,0.000", Scale="1,1,1") {
    rotationMode=BMM_Additive
    rotationSpace=BCS_BoneSpace
    translationMode=BMM_Additive
    translationSpace=BCS_BoneSpace
    alphaBoolBlend {blendOption=Linear}
    alphaScaleBiasClamp {bInterpResult=true}
    boneToModify {boneName=dyn_hairlong_bk_02_r}
    rotation {pitch=15}
   }
   "Transform (Modify) Bone - Bone: dyn_hairlong_bk_02_l"(ComponentPose=@n2, Alpha=1, Translation="0,0,0", Rotation="15,0.000,0.000", Scale="1,1,1") {
    rotationMode=BMM_Additive
    rotationSpace=BCS_BoneSpace
    translationMode=BMM_Additive
    translationSpace=BCS_BoneSpace
    alphaBoolBlend {blendOption=Linear}
    alphaScaleBiasClamp {bInterpResult=true}
    boneToModify {boneName=dyn_hairlong_bk_02_l}
    rotation {pitch=15}
   }
   n3 = Get bIsPlayingMeleeAnim()
   n4 = "Transform (Modify) Bone - Bone: dyn_hairlong_bk_01_MSTR"!(ComponentPose, bAlphaBoolEnabled:bool=@n3, Translation="0,0,0", Rotation="25,-0.000,-0.000", Scale="1,1,1") -> (Pose) {
    alphaInputType=Bool
    rotationMode=BMM_Additive
    rotationSpace=BCS_BoneSpace
    translationMode=BMM_Additive
    translationSpace=BCS_BoneSpace
    alphaBoolBlend {blendInTime=0.1, blendOption=Linear, blendOutTime=0.1}
    alphaScaleBiasClamp {bInterpResult=true}
    boneToModify {boneName=dyn_hairlong_bk_01_MSTR}
    rotation {pitch=25}
   }
   n5 = "Transform (Modify) Bone - Bone: dyn_hairlong_bk_01_r_MSTR"!(ComponentPose=@n4, bAlphaBoolEnabled:bool=@n3, Translation="0,0,0", Rotation="25,5,6.881", Scale="1,1,1") -> (Pose) {
    alphaInputType=Bool
    rotationMode=BMM_Additive
    rotationSpace=BCS_BoneSpace
    translationMode=BMM_Additive
    translationSpace=BCS_BoneSpace
    alphaBoolBlend {blendInTime=0.1, blendOption=Linear, blendOutTime=0.1}
    alphaScaleBiasClamp {bInterpResult=true}
    boneToModify {boneName=dyn_hairlong_bk_01_r_MSTR}
    rotation=<rot 25 5 6.881>
   }
   "Transform (Modify) Bone - Bone: dyn_hairlong_bk_01_l_MSTR"!(ComponentPose=@n5, bAlphaBoolEnabled:bool=@n3, Translation="0,0,0", Rotation="25,-5,0.000", Scale="1,1,1") -> (Pose) {
    alphaInputType=Bool
    rotationMode=BMM_Additive
    rotationSpace=BCS_BoneSpace
    translationMode=BMM_Additive
    translationSpace=BCS_BoneSpace
    alphaBoolBlend {blendInTime=0.1, blendOption=Linear, blendOutTime=0.1}
    alphaScaleBiasClamp {bInterpResult=true}
    boneToModify {boneName=dyn_hairlong_bk_01_l_MSTR}
    rotation {pitch=25, yaw=-5}
   }
   n6 = RigidBody!(
    ComponentPose=Local To Component(LocalPose=Use cached pose 'preSimCache'())
    Alpha=Get isSkydiving() -> (isSkydiving:double)
    OverrideWorldGravity:Vector=Get WindVector() -> (WindVector:Vector)
    ExternalForce
   ) -> (Pose) {
    bClampLinearTranslationLimitToRefPose=true
    bOverrideWorldGravity=true
    overridePhysicsAsset=&/M_MED_Hightower_Tapas_FaceAcc_Physics
    simulationSpace=BaseBoneSpace
    alphaBoolBlend {blendOption=Linear}
    baseBoneRef {boneName=neck_01}
    componentAppliedLinearAccClamp=<3000 3000 3000>
    componentLinearAccScale=<0.3 0.3 0.3>
    componentLinearVelScale=<0.3 0.3 0.35>
    overrideWorldGravity {z=-980}
   }
   n7 = Get IsWearingBackblingAlpha() -> (IsWearingBackblingAlpha:double)
   n8 = "Transform (Modify) Bone - Bone: dyn_hairlong_bk_04_r"(ComponentPose=@n6, Alpha=@n7, Translation="0,0,0", Rotation="7,0.000,0.000", Scale="1,1,1") {
    rotationMode=BMM_Additive
    rotationSpace=BCS_BoneSpace
    translationMode=BMM_Additive
    translationSpace=BCS_BoneSpace
    alphaBoolBlend {blendOption=Linear}
    alphaScaleBiasClamp {bInterpResult=true}
    boneToModify {boneName=dyn_hairlong_bk_04_r}
    rotation {pitch=7}
   }
   n9 = "Transform (Modify) Bone - Bone: dyn_hairlong_bk_04"(ComponentPose=@n8, Alpha=@n7, Translation="-0.213,-0.006,2.025", Rotation="5,0.000,0.000", Scale="1,1,1") {
    rotationMode=BMM_Additive
    rotationSpace=BCS_BoneSpace
    translationMode=BMM_Additive
    translationSpace=BCS_BoneSpace
    alphaBoolBlend {blendOption=Linear}
    alphaScaleBiasClamp {bInterpResult=true}
    boneToModify {boneName=dyn_hairlong_bk_04}
    rotation {pitch=5}
    translation=<-0.213 -0.006 2.025>
   }
   n10 = "Transform (Modify) Bone - Bone: dyn_hairlong_bk_04_l"(ComponentPose=@n9, Alpha=@n7, Translation="0,0,0", Rotation="7,0.000,0.000", Scale="1,1,1") {
    rotationMode=BMM_Additive
    rotationSpace=BCS_BoneSpace
    translationMode=BMM_Additive
    translationSpace=BCS_BoneSpace
    alphaBoolBlend {blendOption=Linear}
    alphaScaleBiasClamp {bInterpResult=true}
    boneToModify {boneName=dyn_hairlong_bk_04_l}
    rotation {pitch=7}
   }
   n11 = Get HeadLookUpAlpha()
   n12 = "Transform (Modify) Bone - Bone: dyn_middle_hair_01_r"(ComponentPose=@n10, Alpha=@n11, Translation="-0.000,0.000,2.789", Rotation="0,0,0", Scale="1,1,1") {
    rotationMode=BMM_Additive
    rotationSpace=BCS_BoneSpace
    translationMode=BMM_Additive
    translationSpace=BCS_BoneSpace
    alphaBoolBlend {blendOption=Linear}
    boneToModify {boneName=dyn_middle_hair_01_r}
    translation {z=2.789}
   }
   n13 = "Transform (Modify) Bone - Bone: dyn_hairlong_bk_02_r"(ComponentPose=@n12, Alpha=@n11, Translation="0,0,0", Rotation="15.000,0.000,0.000", Scale="1,1,1") {
    rotationMode=BMM_Additive
    rotationSpace=BCS_BoneSpace
    translationMode=BMM_Additive
    translationSpace=BCS_BoneSpace
    alphaBoolBlend {blendOption=Linear}
    boneToModify {boneName=dyn_hairlong_bk_02_r}
    rotation {pitch=15}
   }
   n14 = Get InvHeadLookUpAlpha() -> (InvHeadLookUpAlpha:float)
   n15 = "Transform (Modify) Bone - Bone: dyn_hairlong_bk_03_r"(ComponentPose=@n13, Alpha=@n14, Translation="-0.029,0.002,0.350", Rotation="10.000,0.000,-0.001", Scale="1,1,1") {
    rotationMode=BMM_Additive
    rotationSpace=BCS_BoneSpace
    translationMode=BMM_Additive
    translationSpace=BCS_BoneSpace
    alphaBoolBlend {blendOption=Linear}
    boneToModify {boneName=dyn_hairlong_bk_03_r}
    rotation=<rot 10 0 -0.001>
    translation=<-0.029 0.002 0.35>
   }
   n16 = "Transform (Modify) Bone - Bone: dyn_hairlong_bk_04_r"(ComponentPose=@n15, Alpha=@n14, Translation="0,0,0", Rotation="-25.000,0.000,0.002", Scale="1,1,1") {
    rotationMode=BMM_Additive
    rotationSpace=BCS_BoneSpace
    translationMode=BMM_Additive
    translationSpace=BCS_BoneSpace
    alphaBoolBlend {blendOption=Linear}
    boneToModify {boneName=dyn_hairlong_bk_04_r}
    rotation {pitch=-25, roll=0.002}
   }
   n17 = "Transform (Modify) Bone - Bone: dyn_hairlong_bk_04_r"(ComponentPose=@n16, Alpha=@n11, Translation="0,0,0", Rotation="-35.000,0.004,-0.004", Scale="1,1,1") {
    rotationMode=BMM_Additive
    rotationSpace=BCS_BoneSpace
    translationMode=BMM_Additive
    translationSpace=BCS_BoneSpace
    alphaBoolBlend {blendOption=Linear}
    boneToModify {boneName=dyn_hairlong_bk_04_r}
    rotation=<rot -35 0.004 -0.004>
   }
   n18 = Look At - Bone: dyn_hairfront_bt_02_l(ComponentPose=Local To Component(LocalPose=Use cached pose 'preSimCache'()), Alpha=1) {
    alphaBoolBlend {blendOption=Linear}
    boneToModify {boneName=dyn_hairfront_bt_02_l}
    lookAtLocation {x=0}
    lookAtTarget {boneReference {boneName=dyn_hairfront_bt_02_l_aim}}
    lookAt_Axis {axis=<1 0>}
    lookUp_Axis {axis {x=0, z=1}}
   }
   n19 = Look At - Bone: dyn_hairfront_bt_02_r(ComponentPose=@n18, Alpha=1) {
    alphaBoolBlend {blendOption=Linear}
    boneToModify {boneName=dyn_hairfront_bt_02_r}
    lookAtLocation {x=0}
    lookAtTarget {boneReference {boneName=dyn_hairfront_bt_02_r_aim}}
    lookAt_Axis {axis=<1 0>}
    lookUp_Axis {axis {x=0, z=1}}
   }
   Output Pose(Result=Blend Poses by bool(
     BlendPose_0=Component To Local(ComponentPose=@n19)
     BlendPose_1=Component To Local(ComponentPose=@n17)
     bActiveValue=Get bRigidsEnabled() -> (bRigidsEnabled:bool)
    ) {blendType=Linear}) {layerGroup=None}
   n20 = Get HeadLookUpAlpha()
   n21 = Look At - Bone: dyn_hairlong_bk_01_l_MSTR!(ComponentPose=Local To Component(LocalPose=Copy Pose From Mesh() {bUseAttachedParent=true}), Alpha=@n20, LookAtLocation:Vector="0,0,0") -> (Pose) {
    alphaBoolBlend {blendOption=Linear}
    boneToModify {boneName=dyn_hairlong_bk_01_l_MSTR}
    lookAtLocation {x=0, z=10}
    lookAtTarget {boneReference {boneName=dyn_hairlong_bk_01_l_aim}}
    lookAt_Axis {axis=<1 0>}
    lookUp_Axis {axis {x=0, z=1}}
   }
   n22 = Look At - Bone: dyn_hairlong_bk_01_MSTR!(ComponentPose=@n21, Alpha=@n20, LookAtLocation:Vector="0,0,0") -> (Pose) {
    alphaBoolBlend {blendOption=Linear}
    boneToModify {boneName=dyn_hairlong_bk_01_MSTR}
    lookAtLocation {x=0, z=10}
    lookAtTarget {boneReference {boneName=dyn_hairlong_bk_01_aim}}
    lookAt_Axis {axis=<1 0>}
    lookUp_Axis {axis {x=0, z=1}}
   }
   n23 = Look At - Bone: dyn_hairlong_bk_01_r_MSTR!(ComponentPose=@n22, Alpha=@n20, LookAtLocation:Vector="0,0,0") -> (Pose) {
    alphaBoolBlend {blendOption=Linear}
    boneToModify {boneName=dyn_hairlong_bk_01_r_MSTR}
    lookAtLocation {x=0, z=10}
    lookAtTarget {boneReference {boneName=dyn_hairlong_bk_01_r_aim}}
    lookAt_Axis {axis=<1 0>}
    lookUp_Axis {axis {x=0, z=1}}
   }
   n24 = "Constraint - dyn_hairlong_05_r (P)"(ComponentPose=@n23, Alpha=@n20, ConstraintWeights_0=1) {
    alphaBoolBlend {blendOption=Linear}
    boneToModify {boneName=dyn_hairlong_05_r}
    constraintSetup [{targetBone {boneName=dyn_hairfront_bt_02_r}}]
    constraintWeights [1]
   }
   n25 = "Constraint - dyn_hairlong_bk_06_r (PPPPPP)"!(
    ComponentPose=@n24
    Alpha=1
    ConstraintWeights_0=0.0022
    ConstraintWeights_1:float=0.0715
    ConstraintWeights_2:float=0.3455
    ConstraintWeights_3:float=0.5161
    ConstraintWeights_4:float=0.0543
    ConstraintWeights_5:float=0.0104
   ) -> (Pose) {
    alphaBoolBlend {blendOption=Linear}
    boneToModify {boneName=dyn_hairlong_bk_06_r}
    constraintSetup [
     {targetBone {boneName=upperarm_twist_01_r}}
     {targetBone {boneName=spine_03}}
     {targetBone {boneName=spine_05}}
     {targetBone {boneName=clavicle_r}}
     {targetBone {boneName=neck_02}}
     {targetBone {boneName=neck_01}}
    ]
    constraintWeights [1,1,1,1,1,1]
   }
   Save cached pose 'preSimCache'(Pose=Component To Local(ComponentPose=@n25)) {cacheName=preSimCache}
   Get IsPickaxingAlpha() -> (IsPickaxingAlpha:float)
  ]}
 EventGraph {nodes [
   Event BlueprintUpdateAnimation()
   > Set isSkydiving(isSkydiving:double=SelectFloat(A=0.7, B=1, bPickA=Get bIsSkyDiving())) -> (Output_Get:double)
   Event OnBeginTargeting() {is=title K2Node_Event}
   > Set IsTargetingAlpha(IsTargetingAlpha:double=1) -> (Output_Get:double)
   Event OnEndTargeting() {is=title K2Node_Event}
   > Set IsTargetingAlpha(IsTargetingAlpha=0)
   Event OnEndSkydiving()
   > ResetDynamics(InTeleportType=TeleportPhysics)
   > ResetDynamics(InTeleportType=ResetPhysics)
   Event OnCharacterCustomizationLoaded()
   > Set IsWearingBackblingAlpha(IsWearingBackblingAlpha:double="To Float (Boolean)"(InBool=CheckForBackbling())) -> (Output_Get:double)
  ]}
]
variables [
 isSkydiving {type=number}
 IsTargetingAlpha {type=number}
 IsWearingBackblingAlpha {type=number}
]
